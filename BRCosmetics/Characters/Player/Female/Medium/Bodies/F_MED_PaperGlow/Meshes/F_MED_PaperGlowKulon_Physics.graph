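# Physics asset for the PaperGlow female medium body's Kulon (pendant) mesh, defining bodies for a dangling amulet and front skirt bones.
# Capsules on the pelvis and both thighs provide leg and hip volumes alongside the two-link skirt chain.
# The amulet chains from pelvis through a capsule and box to three capsule tassel bones, with angular limits below its root.
# Skirt, thigh and amulet-root constraints leave swing and twist free.
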PhysicsAsset /BRCosmetics/Characters/Player/Female/Medium/Bodies/F_MED_PaperGlow/Meshes/F_MED_PaperGlowKulon_Physics
bodies {
 Dyn_amulet_01 [{
   shapeType=Capsule
   length=1.8402
   radius=1.1611
   shapeName=Dyn_amulet_01_capsule
   center=<0.231 0.0008 0.1315>
   rotation=<rot -2.5902 -90 0.1982>
  }]
 Dyn_amulet_02 [{shapeType=Box, extentX=15.0992, extentY=2.392, extentZ=9.3708, shapeName=Dyn_amulet_02_box, center=<9.0617 -0.1891 0>}]
 Dyn_amulet_10 [{shapeType=Capsule, length=6.9549, radius=1.219, shapeName=Dyn_amulet_10_capsule, center=<3.3266 0 0>, rotation=<rot 90 0 0>}]
 Dyn_amulet_11 [{shapeType=Capsule, length=7.1014, radius=1.5008, shapeName=Dyn_amulet_11_capsule, center=<3.3266 0 0>, rotation=<rot 90 0 0>}]
 Dyn_amulet_12 [{shapeType=Capsule, length=6.8791, radius=0.8013, shapeName=Dyn_amulet_12_capsule, center=<3.3266 0 0>, rotation=<rot 90 0 0>}]
 Dyn_skirt_Front_01 [{shapeType=Capsule, length=10.9553, radius=4.4332, shapeName=Dyn_skirt_Front_01_capsule, center=<8.8433 0 0>, rotation=<rot 90 0 0>}]
 Dyn_skirt_Front_02 [{shapeType=Capsule, length=10.2951, radius=5.0934, shapeName=Dyn_skirt_Front_02_capsule, center=<8.8433 0 0>, rotation=<rot 90 0 0>}]
 pelvis [{shapeType=Capsule, length=15.5705, radius=8.0724, shapeName=pelvis_capsule, center=<0 0 0.763>, rotation=<rot 0 -0.0001 5>}]
 thigh_l [{shapeType=Capsule, length=40.1712, radius=9.5001, shapeName=thigh_r_capsule, center=<-22.0662 -1.7793 -1.0864>, rotation=<rot 90 0 180>}]
 thigh_r [{
   shapeType=Capsule
   length=26.1655
   radius=12.1545
   shapeName=thigh_r_capsule
   center=<19.2489 -1.1629 -1.4435>
   rotation=<rot 87.7552 0.0001 0.0001>
  }]
}
constraints [
 {
  bone1Name=thigh_l
  bone2Name=pelvis
  swing1LimitDegrees=45
  swing1Motion=Free
  swing2LimitDegrees=45
  swing2Motion=Free
  twistLimitDegrees=45
  twistMotion=Free
 }
 {
  bone1Name=thigh_r
  bone2Name=pelvis
  swing1LimitDegrees=45
  swing1Motion=Free
  swing2LimitDegrees=45
  swing2Motion=Free
  twistLimitDegrees=45
  twistMotion=Free
 }
 {
  bone1Name=Dyn_skirt_Front_01
  bone2Name=pelvis
  swing1LimitDegrees=45
  swing1Motion=Free
  swing2LimitDegrees=45
  swing2Motion=Free
  twistLimitDegrees=45
  twistMotion=Free
 }
 {
  bone1Name=Dyn_skirt_Front_02
  bone2Name=Dyn_skirt_Front_01
  swing1LimitDegrees=45
  swing1Motion=Free
  swing2LimitDegrees=45
  swing2Motion=Free
  twistLimitDegrees=45
  twistMotion=Free
 }
 {bone1Name=Dyn_amulet_10, bone2Name=Dyn_amulet_02, swing1LimitDegrees=22, swing2LimitDegrees=15, twistLimitDegrees=0}
 {bone1Name=Dyn_amulet_11, bone2Name=Dyn_amulet_02, swing1LimitDegrees=22, swing2LimitDegrees=15, twistLimitDegrees=0}
 {bone1Name=Dyn_amulet_12, bone2Name=Dyn_amulet_02, swing1LimitDegrees=22, swing2LimitDegrees=15, twistLimitDegrees=0}
 {bone1Name=Dyn_amulet_02, bone2Name=Dyn_amulet_01, swing1LimitDegrees=50, swing2LimitDegrees=5, twistLimitDegrees=5}
 {
  bone1Name=Dyn_amulet_01
  bone2Name=pelvis
  swing1LimitDegrees=45
  swing1Motion=Free
  swing2LimitDegrees=45
  swing2Motion=Free
  twistLimitDegrees=45
  twistMotion=Free
 }
]
