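# Physics asset for the cape of the PaperGlow female medium body, covering left and right corrective cape bones.
# Capsules on the upper spine and upper arms form the body volume, with tiny capsules on each cape base and longer ones on the Dyn cape bones.
# Constraints chain spine to arms to cape bases with free rotation, while each Dyn bone has narrow swing and twist limits against its base.
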PhysicsAsset /BRCosmetics/Characters/Player/Female/Medium/Bodies/F_MED_PaperGlow/Meshes/F_MED_PaperGlow_Cape_Physics
bodies {
 Base_Correct_cape_L [{shapeType=Capsule, length=0.505, radius=0.505, shapeName=Base_Correct_cape_L_capsule, rotation=<rot 0 -90 0>}]
 Base_Correct_cape_R [{shapeType=Capsule, length=0.505, radius=0.505, shapeName=Base_Correct_cape_R_capsule, rotation=<rot 0 -90 0>}]
 Dyn_Correct_cape_L [{
   shapeType=Capsule
   length=10.3685
   radius=3.0365
   shapeName=Dyn_Correct_cape_L_capsule
   center=<3.036 7.649 0>
   rotation=<rot 81.833 7.9357 -14.1133>
  }]
 Dyn_Correct_cape_R [{
   shapeType=Capsule
   length=11.0881
   radius=2.8046
   shapeName=Dyn_Correct_cape_R_capsule
   center=<-1.5073 -7.4397 0>
   rotation=<rot -0.3839 -64.1971 -83.3196>
  }]
 spine_03 [{shapeType=Capsule, length=6.139, radius=9.0425, shapeName=spine_03_capsule}]
 spine_04 [{shapeType=Capsule, length=4.3429, radius=12.8975, shapeName=spine_04_capsule, center=<0 1.1097 0>, rotation=<rot -90 90 -90.0003>}]
 spine_05 [{shapeType=Capsule, length=32.7128, radius=6.8844, shapeName=spine_05_capsule, center=<-2.9776 -1.6809 0>}]
 upperarm_l [{
   shapeType=Capsule
   length=14.8878
   radius=5.4164
   shapeName=upperarm_l_capsule
   center=<7.7221 -0.7233 0>
   rotation=<rot -90 179.9802 -179.9802>
  }]
 upperarm_r [{shapeType=Capsule, length=14.7161, radius=5.544, shapeName=upperarm_r_capsule, center=<-8.336 0 0>, rotation=<rot 90 0 0>}]
}
constraints [
 {
  bone1Name=spine_04
  bone2Name=spine_03
  swing1LimitDegrees=45
  swing1Motion=Free
  swing2LimitDegrees=45
  swing2Motion=Free
  twistLimitDegrees=45
  twistMotion=Free
 }
 {
  bone1Name=spine_05
  bone2Name=spine_04
  swing1LimitDegrees=45
  swing1Motion=Free
  swing2LimitDegrees=45
  swing2Motion=Free
  twistLimitDegrees=45
  twistMotion=Free
 }
 {
  bone1Name=upperarm_l
  bone2Name=spine_05
  swing1LimitDegrees=45
  swing1Motion=Free
  swing2LimitDegrees=45
  swing2Motion=Free
  twistLimitDegrees=45
  twistMotion=Free
 }
 {
  bone1Name=Base_Correct_cape_L
  bone2Name=upperarm_l
  swing1LimitDegrees=45
  swing1Motion=Free
  swing2LimitDegrees=45
  swing2Motion=Free
  twistLimitDegrees=45
  twistMotion=Free
 }
 {bone1Name=Dyn_Correct_cape_L, bone2Name=Base_Correct_cape_L, swing1LimitDegrees=15, swing2LimitDegrees=15, twistLimitDegrees=15}
 {bone1Name=Dyn_Correct_cape_R, bone2Name=Base_Correct_cape_R, swing1LimitDegrees=16, swing2LimitDegrees=16, twistLimitDegrees=16}
 {
  bone1Name=upperarm_r
  bone2Name=spine_05
  swing1LimitDegrees=45
  swing1Motion=Free
  swing2LimitDegrees=45
  swing2Motion=Free
  twistLimitDegrees=45
  twistMotion=Free
 }
 {
  bone1Name=Base_Correct_cape_R
  bone2Name=upperarm_r
  swing1LimitDegrees=45
  swing1Motion=Free
  swing2LimitDegrees=45
  swing2Motion=Free
  twistLimitDegrees=45
  twistMotion=Free
 }
]
