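# Control Rig for the PaperGlow female medium player body that applies corrective deformation to the cosmetic's skeleton.
# The forward solve makes one call to the shared DeformRig function, configured with per-region toggles, transform offsets, curve flags and debug switches.
# The hierarchy pairs deform bones with base parents at shoulders, elbows, wrists, pecs, groin, glutes and knees.
# It also holds Dyn_ cloth and accessory bones, with no simulation nodes in this graph.
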ControlRigBlueprint /BRCosmetics/Characters/Player/Female/Medium/Bodies/F_MED_PaperGlow/Meshes/F_MED_PaperGlow_Deform_CtrlRig
graphs [
 RigVMFunctionLibrary {links [], nodes []}
 RigVMModel {nodes [
   RigUnit_BeginExecution = &RigUnit_BeginExecution()
   FN_DeformRig_FNC = RigVMFunctionReferenceNode(
    DeformRig_ Curve="(deformRig_l_upperarm_up_alpha_72_#1=False,deformRig_r_upperarm_up_alpha_73_#2=False,deformRig_l_upperarm_fr_alpha_74_#3=False,deformRig_r_upperarm_fr_alpha_75_CB51B8 ...(truncated)"
    DeformRig_Bool="(upperarm_1_#4=True,elbow_3_#5=True,wrist_7_#6=True,groin_8_#7=True,glute_17_#8=True,knee_9_40C408324CA261 ...(truncated)"
    DeformRig_Debug=(upperarm_debug_18_#4=False,wrist_debug_20_#6=False,lat_debug_26_#9=False,groinGlute_debug_27_#7=False)
    DeformRig_Value="(upperarm_up_18_#10=(Rotation=(X=0,Y=0,Z=0,W=1),Translation=(X=0,Y=0,Z=8),Scale3D=(X=0.7,Y=1,Z=1)),upperarm_fwd_fr_19_#11=(Rotation=(X=0,Y=0,Z=0,W=1),Translation=(X=0,Y=10,Z=0),Scale3D=(X=0.5,Y=1,Z= ...(truncated)"
    ExecuteContext=@RigUnit_BeginExecution.ExecutePin
   )
  ]}
]
hierarchy [
 root {loc=<0 0 0>, rot=<0 0 0 1>}
 pelvis {parent=root, loc=<0 -0.5301 98.0571>, rot=<0.0127 0.707 -0.0127 -0.707>}
 spine_01 {parent=pelvis, loc=<6.1208 -0.9328 0>, rot=<0 0 0.034 -0.9994>}
 spine_02 {parent=spine_01, loc=<6.5996 -0.2095 0>, rot=<0 0 0.0238 -0.9997>}
 spine_03 {parent=spine_02, loc=<6.4914 -0.1237 0>, rot=<0 0 0.0149 -0.9999>}
 spine_04 {parent=spine_03, loc=<6.9598 0.0794 0>, rot=<0 0 -0.0132 -0.9999>}
 spine_05 {parent=spine_04, loc=<6.9987 0 0>, rot=<0 0 -0.0015 -1>}
 attach_cape {parent=spine_05, loc=<9.0205 -11.7142 -0.0004>, rot=<0.0616 -0.7044 -0.0615 -0.7045>}
 neck_01 {parent=spine_05, loc=<11.3138 0.1529 0>, rot=<0 0 -0.2003 -0.9797>}
 neck_02 {parent=neck_01, loc=<4.5627 0 0>, rot=<0 0 0.0504 -0.9987>}
 head {parent=neck_02, loc=<5.3791 0.0006 -0.0024>, rot=<0 0 0.115 -0.9934>}
 attach_backpack {parent=spine_05, loc=<-0.118 -13.4731 -0.0005>, rot=<-0.0093 -0.707 0.0094 -0.7071>}
 clavicle_l {parent=spine_05, loc=<5.4831 -0.1053 -3.2946>, rot=<0.0465 -0.7289 -0.009 -0.6829>}
 upperarm_l {parent=clavicle_l, loc=<9.8658 -1.2136 -0.4105>, rot=<0.0164 -0.3245 -0.0269 -0.9454>}
 lowerarm_l {parent=upperarm_l, loc=<26.465 -0.8345 0.259>, rot=<0.0036 0.0161 -0.2298 -0.9731>}
 hand_l {parent=lowerarm_l, loc=<22.7355 -0.3203 -0.6105>, rot=<-0.6784 -0.0223 0.0735 -0.7307>}
 thumb_01_l {parent=hand_l, loc=<1.3118 -0.3191 -2.459>, rot=<0.6546 -0.4042 0.0664 -0.6354>}
 thumb_02_l {parent=thumb_01_l, loc=<5.9336 -0.0448 0.0401>, rot=<-0.0037 0.0213 0.298 -0.9543>}
 thumb_03_l {parent=thumb_02_l, loc=<2.2223 0.3682 -0.0907>, rot=<0.0118 -0.013 0.0066 -0.9998>}
 index_metacarpal_l {parent=hand_l, loc=<1.6535 0.0183 -2.96>, rot=<0.0158 -0.0124 -0.0173 -0.9996>}
 index_01_l {parent=index_metacarpal_l, loc=<8.3008 0.0357 -0.0524>, rot=<-0.0044 0.0485 0.0924 -0.9945>}
 index_02_l {parent=index_01_l, loc=<3.3587 -0.3004 0.0452>, rot=<-0.002 0.0027 0.3225 -0.9466>}
 index_03_l {parent=index_02_l, loc=<2.5578 -0.123 -0.0252>, rot=<-0.002 0.0067 0.2558 -0.9667>}
 middle_metacarpal_l {parent=hand_l, loc=<2.1856 -0.1725 -0.5984>, rot=<0.006 0.0168 -0.0329 -0.9993>}
 middle_01_l {parent=middle_metacarpal_l, loc=<7.3275 -0.2447 0.0129>, rot=<-0.0208 0.0042 0.0858 -0.9961>}
 middle_02_l {parent=middle_01_l, loc=<4.0123 0.1401 0.1175>, rot=<-0.0218 0.0258 0.3512 -0.9357>}
 middle_03_l {parent=middle_02_l, loc=<2.6077 0.0534 0.1373>, rot=<0.0082 0.0188 0.1708 -0.9851>}
 ring_metacarpal_l {parent=hand_l, loc=<2.2074 -0.3625 1.3638>, rot=<-0.0284 0.0722 -0.0458 -0.9959>}
 ring_01_l {parent=ring_metacarpal_l, loc=<7.0295 -0.6197 -0.1446>, rot=<-0.0489 -0.0252 0.1892 -0.9804>}
 ring_02_l {parent=ring_01_l, loc=<3.6174 0.4647 -0.1027>, rot=<0 0 0.2806 -0.9598>}
 ring_03_l {parent=ring_02_l, loc=<2.4152 0.114 0.0199>, rot=<0 0 0.1189 -0.9929>}
 pinky_metacarpal_l {parent=hand_l, loc=<1.7616 -0.9753 3.3814>, rot=<-0.0397 0.0814 -0.069 -0.9935>}
 pinky_01_l {parent=pinky_metacarpal_l, loc=<7.1709 -0.4831 -0.1094>, rot=<-0.0654 -0.0035 0.2242 -0.9723>}
 pinky_02_l {parent=pinky_01_l, loc=<2.8274 0.0893 -0.1043>, rot=<0.0037 0.0011 0.2854 -0.9584>}
 pinky_03_l {parent=pinky_02_l, loc=<2.2046 0.0307 -0.0078>, rot=<-0.0002 0.0098 0.1445 -0.9895>}
 deform_wrist_dn_base_l {parent=hand_l, loc=<0 -1.5326 0>, rot=<0 0 0 -1>}
 deform_wrist_dn_l {parent=deform_wrist_dn_base_l, loc=<0 0 0>, rot=<0 0 0 1>}
 deform_wrist_up_base_l {parent=hand_l, loc=<0 2.3009 0>, rot=<0 0 0 -1>}
 deform_wrist_up_l {parent=deform_wrist_up_base_l, loc=<0 0 0>, rot=<0 0 0 1>}
 Dyn_bracelet_L_01 {parent=hand_l, loc=<-0.517 0.77 -0.2584>, rot=<0.7846 0.0148 -0.0547 -0.6174>}
 lowerarm_twist_02_l {parent=lowerarm_l, loc=<10.6707 0 0>, rot=<0 0 0 1>}
 lowerarm_twist_01_l {parent=lowerarm_l, loc=<18.316 0 0>, rot=<0 0 0 1>}
 deform_elbow_in_base_l {parent=lowerarm_l, loc=<0.1834 3.6062 -0.2572>, rot=<0 0 0 -1>}
 deform_elbow_in_l {parent=deform_elbow_in_base_l, loc=<0 0 0>, rot=<0 0 0 1>}
 deform_elbow_out_base_l {parent=lowerarm_l, loc=<0 -4 0>, rot=<0 0 0 1>}
 deform_elbow_out_l {parent=deform_elbow_out_base_l, loc=<0 0 0>, rot=<0 0 0 1>}
 upperarm_twist_01_l {parent=upperarm_l, loc=<7.2135 -0.1134 0>, rot=<0 0 0 1>}
 deform_upperarm_bk_base_l {parent=upperarm_twist_01_l, loc=<-6.4053 -4.3746 0.314>, rot=<0 0 0 -1>}
 deform_upperarm_bk_l {parent=deform_upperarm_bk_base_l, loc=<0 0 0>, rot=<0 0 0 1>}
 deform_upperarm_fr_base_l {parent=upperarm_twist_01_l, loc=<-5.6312 5.092 0.3354>, rot=<0 0 0 -1>}
 deform_upperarm_fr_l {parent=deform_upperarm_fr_base_l, loc=<0 0 0>, rot=<0 0 0 1>}
 deform_upperarm_up_base_l {parent=upperarm_twist_01_l, loc=<-8.891 0.4825 3.9517>, rot=<0 0 0 -1>}
 deform_upperarm_up_l {parent=deform_upperarm_up_base_l, loc=<0 0 0>, rot=<0 0 0 1>}
 Base_Correct_cape_L {parent=upperarm_twist_01_l, loc=<0 0 0>, rot=<0 0 0 1>}
 Dyn_Correct_cape_L {parent=Base_Correct_cape_L, loc=<1.0748 6.7742 -0.0056>, rot=<0.5362 -0.0582 -0.0378 -0.8413>}
 upperarm_twist_02_l {parent=upperarm_l, loc=<19.7998 -0.1134 0>, rot=<0 0 0 1>}
 clavicle_r {parent=spine_05, loc=<5.4831 -0.1053 3.2946>, rot=<0.7289 0.0465 -0.6829 0.009>}
 upperarm_r {parent=clavicle_r, loc=<-9.8658 1.2136 0.4105>, rot=<0.0164 -0.3245 -0.0269 -0.9454>}
 lowerarm_r {parent=upperarm_r, loc=<-26.465 0.8345 -0.259>, rot=<0.0036 0.0161 -0.2298 -0.9731>}
 hand_r {parent=lowerarm_r, loc=<-22.7355 0.3203 0.6106>, rot=<-0.6784 -0.0223 0.0735 -0.7307>}
 weapon_r {parent=hand_r, loc=<-7.826 2.664 0.279>, rot=<0.0182 0.0416 0.0388 -0.9982>}
 thumb_01_r {parent=hand_r, loc=<-1.3118 0.3191 2.459>, rot=<0.6546 -0.4042 0.0664 -0.6354>}
 thumb_02_r {parent=thumb_01_r, loc=<-5.9336 0.0448 -0.0401>, rot=<-0.0037 0.0213 0.298 -0.9543>}
 thumb_03_r {parent=thumb_02_r, loc=<-2.2223 -0.3682 0.0907>, rot=<0.0118 -0.013 0.0066 -0.9998>}
 index_metacarpal_r {parent=hand_r, loc=<-1.6535 -0.0183 2.96>, rot=<0.0158 -0.0124 -0.0173 -0.9996>}
 index_01_r {parent=index_metacarpal_r, loc=<-8.3008 -0.0357 0.0524>, rot=<-0.0044 0.0485 0.0924 -0.9945>}
 index_02_r {parent=index_01_r, loc=<-3.3587 0.3004 -0.0452>, rot=<-0.002 0.0027 0.3225 -0.9466>}
 index_03_r {parent=index_02_r, loc=<-2.5578 0.123 0.0252>, rot=<-0.002 0.0067 0.2558 -0.9667>}
 middle_metacarpal_r {parent=hand_r, loc=<-2.1856 0.1725 0.5984>, rot=<0.006 0.0168 -0.0329 -0.9993>}
 middle_01_r {parent=middle_metacarpal_r, loc=<-7.3275 0.2447 -0.0129>, rot=<-0.0208 0.0042 0.0858 -0.9961>}
 middle_02_r {parent=middle_01_r, loc=<-4.0123 -0.1401 -0.1175>, rot=<-0.0218 0.0258 0.3512 -0.9357>}
 middle_03_r {parent=middle_02_r, loc=<-2.6077 -0.0534 -0.1373>, rot=<0.0082 0.0188 0.1708 -0.9851>}
 ring_metacarpal_r {parent=hand_r, loc=<-2.2074 0.3625 -1.3638>, rot=<-0.0284 0.0722 -0.0458 -0.9959>}
 ring_01_r {parent=ring_metacarpal_r, loc=<-7.0295 0.6197 0.1446>, rot=<-0.0489 -0.0252 0.1892 -0.9804>}
 ring_02_r {parent=ring_01_r, loc=<-3.6174 -0.4647 0.1027>, rot=<0 0 0.2806 -0.9598>}
 ring_03_r {parent=ring_02_r, loc=<-2.4152 -0.114 -0.0199>, rot=<0 0 0.1189 -0.9929>}
 pinky_metacarpal_r {parent=hand_r, loc=<-1.7616 0.9753 -3.3814>, rot=<-0.0397 0.0814 -0.069 -0.9935>}
 pinky_01_r {parent=pinky_metacarpal_r, loc=<-7.1709 0.4831 0.1094>, rot=<-0.0654 -0.0035 0.2242 -0.9723>}
 pinky_02_r {parent=pinky_01_r, loc=<-2.8274 -0.0893 0.1043>, rot=<0.0037 0.0011 0.2854 -0.9584>}
 pinky_03_r {parent=pinky_02_r, loc=<-2.2046 -0.0307 0.0078>, rot=<-0.0002 0.0098 0.1445 -0.9895>}
 hand_attach_r {parent=hand_r, loc=<-7.8255 2.6641 0.2791>, rot=<-0.0165 0.0423 -0.6784 -0.7333>}
 deform_wrist_dn_base_r {parent=hand_r, loc=<-0.0001 1.5328 0>, rot=<0 0 0 -1>}
 deform_wrist_dn_r {parent=deform_wrist_dn_base_r, loc=<0 0 0>, rot=<0 0 0 1>}
 deform_wrist_up_base_r {parent=hand_r, loc=<-0.0002 -2.3007 -0.0001>, rot=<0 0 0 -1>}
 deform_wrist_up_r {parent=deform_wrist_up_base_r, loc=<0 0 0>, rot=<0 0 0 1>}
 lowerarm_twist_02_r {parent=lowerarm_r, loc=<-10.6707 0 0>, rot=<0 0 0 -1>}
 lowerarm_twist_01_r {parent=lowerarm_r, loc=<-18.316 0 0>, rot=<0 0 0 -1>}
 Dyn_bracelet_R {parent=lowerarm_twist_01_r, loc=<-3.1281 0.0013 0.0012>, rot=<0 0 0 -1>}
 deform_elbow_in_base_r {parent=lowerarm_r, loc=<-0.1834 -3.6062 0.2572>, rot=<0 0 0 -1>}
 deform_elbow_in_r {parent=deform_elbow_in_base_r, loc=<0 0 0>, rot=<0 0 0 1>}
 deform_elbow_out_base_r {parent=lowerarm_r, loc=<0 4 -0.0001>, rot=<0 0 0 -1>}
 deform_elbow_out_r {parent=deform_elbow_out_base_r, loc=<0 0 0>, rot=<0 0 0 1>}
 upperarm_twist_01_r {parent=upperarm_r, loc=<-7.2135 0.1134 0>, rot=<0 0 0 -1>}
 deform_upperarm_bk_base_r {parent=upperarm_twist_01_r, loc=<6.4053 4.3746 -0.314>, rot=<0 0 0 -1>}
 deform_upperarm_bk_r {parent=deform_upperarm_bk_base_r, loc=<0 0 0>, rot=<0 0 0 1>}
 deform_upperarm_fr_base_r {parent=upperarm_twist_01_r, loc=<5.6312 -5.092 -0.3354>, rot=<0 0 0 -1>}
 deform_upperarm_fr_r {parent=deform_upperarm_fr_base_r, loc=<0 0 0>, rot=<0 0 0 1>}
 deform_upperarm_up_base_r {parent=upperarm_twist_01_r, loc=<8.891 -0.4825 -3.9517>, rot=<0 0 0 -1>}
 deform_upperarm_up_r {parent=deform_upperarm_up_base_r, loc=<0 0 0>, rot=<0 0 0 1>}
 Base_Correct_cape_R {parent=upperarm_twist_01_r, loc=<0 0 0>, rot=<0 0 0 1>}
 Dyn_Correct_cape_R {parent=Base_Correct_cape_R, loc=<-1.0749 -6.7742 0.0057>, rot=<0.5362 -0.0582 -0.0378 -0.8413>}
 upperarm_twist_02_r {parent=upperarm_r, loc=<-19.7998 0.1134 0>, rot=<0 0 0 -1>}
 Dyn_pendant_01 {parent=spine_04, loc=<-2.4371 12.7298 0.02>, rot=<0.0266 -0.9996 0 0>}
 Dyn_pendant_02 {parent=Dyn_pendant_01, loc=<2.0008 0 0.0002>, rot=<0 0 0 1>}
 deform_pec_aim_base_l {parent=spine_04, loc=<3.2456 9.3309 -1.2237>, rot=<0.0603 -0.4033 0.3318 -0.8507>}
 deform_pec_aim_l {parent=deform_pec_aim_base_l, loc=<0 0 0>, rot=<0 0 0 1>}
 deform_pec_base_l {parent=deform_pec_aim_l, loc=<9.3665 0 0>, rot=<0 0 0 1>}
 deform_pec_l {parent=deform_pec_base_l, loc=<0 0 0>, rot=<0 0 0 1>}
 deform_pec_aim_base_r {parent=spine_04, loc=<3.2452 9.3309 1.2237>, rot=<-0.4033 -0.0603 0.8507 0.3318>}
 deform_pec_aim_r {parent=deform_pec_aim_base_r, loc=<0 0 0>, rot=<0 0 0 1>}
 deform_pec_base_r {parent=deform_pec_aim_r, loc=<-9.3667 -0.0003 -0.0004>, rot=<0 0 0 1>}
 deform_pec_r {parent=deform_pec_base_r, loc=<0 0 0>, rot=<0 0 0 1>}
 Dyn_cape_back_Base {parent=spine_04, loc=<0.034 -11.098 -3.672>, rot=<0.1384 -0.9849 -0.053 -0.0898>}
 Dyn_cape_back {parent=Dyn_cape_back_Base, loc=<0 0 0>, rot=<0 0 0 1>}
 Dyn_belt_Front_Base {parent=spine_01, loc=<0.375 9.624 0>, rot=<0.5498 0.0096 0.8326 -0.0659>}
 Dyn_belt_Front_01 {parent=Dyn_belt_Front_Base, loc=<0 0 0>, rot=<0 0 0 1>}
 Dyn_belt_Front_02 {parent=Dyn_belt_Front_01, loc=<11.146 0.929 0.289>, rot=<0.1097 0.0891 -0.2328 -0.9622>}
 Dyn_belt_Front_03 {parent=Dyn_belt_Front_02, loc=<8.095 2.928 0.204>, rot=<-0.0212 0.093 0.2816 -0.9548>}
 Dyn_belt_back_Base {parent=spine_01, loc=<3.34 -8.358 0.46>, rot=<0.4824 0.1702 0.8428 -0.1673>}
 Dyn_belt_back_01 {parent=Dyn_belt_back_Base, loc=<0 0 0>, rot=<0 0 0 1>}
 Dyn_belt_back_02 {parent=Dyn_belt_back_01, loc=<10.842 -1.364 1.327>, rot=<-0.0038 0.1534 0.1878 -0.9701>}
 Dyn_belt_back_03 {parent=Dyn_belt_back_02, loc=<8.544 -5.798 1.789>, rot=<0.1717 0.1448 -0.2179 -0.9498>}
 thigh_l {parent=pelvis, loc=<-1.0855 -0.2109 -8.7053>, rot=<0.0009 0.0355 0.0155 -0.9993>}
 calf_l {parent=thigh_l, loc=<-42.5284 -0.0019 -0.7297>, rot=<0.0006 -0.0161 -0.0391 -0.9991>}
 foot_l {parent=calf_l, loc=<-46.7401 -2.5014 -0.592>, rot=<0.0427 -0.0238 0.0353 -0.9982>}
 ball_l {parent=foot_l, loc=<-3.5658 10.8066 -0.1544>, rot=<0.0001 -0.0001 -0.7063 -0.7079>}
 calf_twist_02_l {parent=calf_l, loc=<-19.1546 -1.7943 -0.2638>, rot=<0.0237 0.0095 -0.0035 -0.9997>}
 calf_twist_01_l {parent=calf_l, loc=<-35.8483 -1.9197 -0.5794>, rot=<0.0237 0.0095 -0.0035 -0.9997>}
 deform_knee_in_base_l {parent=calf_l, loc=<0.8031 -5.0231 0.2456>, rot=<0 0 0 -1>}
 deform_knee_in_l {parent=deform_knee_in_base_l, loc=<0 0 0>, rot=<0 0 0 1>}
 deform_knee_out_base_l {parent=calf_l, loc=<0 4 0>, rot=<0 0 0 -1>}
 deform_knee_out_l {parent=deform_knee_out_base_l, loc=<0 0 0>, rot=<0 0 0 1>}
 thigh_twist_01_l {parent=thigh_l, loc=<-27.952 0 0>, rot=<0.0013 0 0 -1>}
 deform_groin_base_l {parent=thigh_l, loc=<0.4801 5.6776 -0.2486>, rot=<0 0 0 -1>}
 deform_groin_l {parent=deform_groin_base_l, loc=<0 0 0>, rot=<0 0 0 1>}
 thigh_r {parent=pelvis, loc=<-1.0855 -0.2109 8.7053>, rot=<0.0355 -0.0009 0.9993 0.0155>}
 calf_r {parent=thigh_r, loc=<42.5284 0.0019 0.7297>, rot=<0.0006 -0.0161 -0.0391 -0.9991>}
 foot_r {parent=calf_r, loc=<46.7401 2.5014 0.592>, rot=<0.0427 -0.0238 0.0353 -0.9982>}
 ball_r {parent=foot_r, loc=<3.5658 -10.8066 0.1544>, rot=<0.0001 -0.0001 -0.7063 -0.7079>}
 calf_twist_02_r {parent=calf_r, loc=<19.1546 1.7943 0.2638>, rot=<0.0237 0.0095 -0.0035 -0.9997>}
 calf_twist_01_r {parent=calf_r, loc=<35.8483 1.9197 0.5794>, rot=<0.0237 0.0095 -0.0035 -0.9997>}
 deform_knee_in_base_r {parent=calf_r, loc=<-0.8031 5.0231 -0.2457>, rot=<0 0 0 -1>}
 deform_knee_in_r {parent=deform_knee_in_base_r, loc=<0 0 0>, rot=<0 0 0 1>}
 deform_knee_out_base_r {parent=calf_r, loc=<0 -4 0>, rot=<0 0 0 -1>}
 deform_knee_out_r {parent=deform_knee_out_base_r, loc=<0 0 0>, rot=<0 0 0 1>}
 thigh_twist_01_r {parent=thigh_r, loc=<27.952 0 0>, rot=<0.0013 0 0 -1>}
 deform_groin_base_r {parent=thigh_r, loc=<-0.4801 -5.6776 0.2486>, rot=<0 0 0 -1>}
 deform_groin_r {parent=deform_groin_base_r, loc=<0 0 0>, rot=<0 0 0 1>}
 deform_glute_base_l {parent=pelvis, loc=<-7.3466 -9.0213 -9.1245>, rot=<0.0172 0.031 -0.4477 -0.8935>}
 deform_glute_l {parent=deform_glute_base_l, loc=<0 0 0>, rot=<0 0 0 1>}
 deform_glute_base_r {parent=pelvis, loc=<-7.3464 -9.0213 9.1245>, rot=<0.031 -0.0172 0.8935 -0.4477>}
 deform_glute_r {parent=deform_glute_base_r, loc=<0 0 0>, rot=<0 0 0 1>}
 Dyn_skirt_L_01_Base {parent=pelvis, loc=<2.5673 -0.0176 -13.2289>, rot=<0.0128 -0.9869 -0.0185 -0.1599>}
 Dyn_skirt_L_01 {parent=Dyn_skirt_L_01_Base, loc=<0 0 0>, rot=<0 0 0 1>}
 Dyn_skirt_L_02 {parent=Dyn_skirt_L_01, loc=<14.5289 0 0>, rot=<0.0006 -0.0336 0.0167 -0.9993>}
 Dyn_skirt_L_03 {parent=Dyn_skirt_L_02, loc=<13.4445 0 0>, rot=<-0.0012 -0.0522 -0.014 -0.9985>}
 Dyn_skirt_L_04 {parent=Dyn_skirt_L_03, loc=<14.8385 0 0>, rot=<0 0 0 1>}
 Dyn_skirt_L_06_Base {parent=pelvis, loc=<2.8329 -8.6354 -9.8971>, rot=<0.043 -0.9074 0.4013 -0.1175>}
 Dyn_skirt_L_06 {parent=Dyn_skirt_L_06_Base, loc=<0 0 0>, rot=<0 0 0 1>}
 Dyn_skirt_L_07 {parent=Dyn_skirt_L_06, loc=<10.5481 0 0>, rot=<0 -0.0221 0 -0.9998>}
 Dyn_skirt_L_08 {parent=Dyn_skirt_L_07, loc=<12.3591 0 0>, rot=<0.0003 -0.0341 0.011 -0.9994>}
 Dyn_skirt_L_09 {parent=Dyn_skirt_L_08, loc=<12.1671 0 0>, rot=<0 0 0 1>}
 Dyn_skirt_back_01_Base {parent=pelvis, loc=<2.4833 -10.4775 0>, rot=<0.0963 -0.7023 0.6993 -0.0919>}
 Dyn_skirt_back_01 {parent=Dyn_skirt_back_01_Base, loc=<0 0 0>, rot=<0 0 0 1>}
 Dyn_skirt_back_02 {parent=Dyn_skirt_back_01, loc=<10.8592 0 0>, rot=<0 -0.0638 0 -0.998>}
 Dyn_skirt_back_03 {parent=Dyn_skirt_back_02, loc=<11.0369 0 0>, rot=<-0.0009 -0.0146 -0.0069 -0.9999>}
 Dyn_skirt_back_04 {parent=Dyn_skirt_back_03, loc=<12.4391 0 0>, rot=<0 0 0 1>}
 Dyn_skirt_Front_01_Base {parent=pelvis, loc=<0.9562 8.2342 -2.3084>, rot=<0.0759 0.6849 0.7247 0.002>}
 Dyn_skirt_Front_01 {parent=Dyn_skirt_Front_01_Base, loc=<0 0 0>, rot=<0 0 0 1>}
 Dyn_skirt_Front_02 {parent=Dyn_skirt_Front_01, loc=<18.1531 0 0>, rot=<0 0 0 1>}
 Dyn_skirt_Parent_Base {parent=pelvis, loc=<-4.1982 8.2539 8.8326>, rot=<0.0673 0.6858 0.7246 0.0111>}
 Dyn_skirt_Parent_01 {parent=Dyn_skirt_Parent_Base, loc=<0 0 0>, rot=<0 0 0 1>}
 Dyn_skirt_Parent_02 {parent=Dyn_skirt_Parent_01, loc=<18.689 0 0>, rot=<0 0 -0.0143 -0.9999>}
 Dyn_skirt_Parent_03 {parent=Dyn_skirt_Parent_02, loc=<19.3071 0 0>, rot=<-0.0008 -0.0194 -0.0069 -0.9998>}
 Dyn_skirt_R_01_Base {parent=pelvis, loc=<-3.142 -3.539 15.5074>, rot=<0.0938 -0.1047 0.9901 -0.0055>}
 Dyn_skirt_R_02 {parent=Dyn_skirt_R_01_Base, loc=<0 0 0>, rot=<0 0 -0.0036 -1>}
 Dyn_skirt_R_03 {parent=Dyn_skirt_R_02, loc=<12.4891 0 0>, rot=<0 -0.0027 -0.0058 -1>}
 Dyn_skirt_R_04 {parent=Dyn_skirt_R_03, loc=<9.5008 0 0>, rot=<0.0073 -0.0089 -0.0092 -0.9999>}
 Dyn_skirt_R_05 {parent=Dyn_skirt_R_04, loc=<9.4469 0 0>, rot=<0.0078 -0.0002 -0.0208 -0.9998>}
 Dyn_amulet_01_Base {parent=pelvis, loc=<7.1036 8.0546 -8.0602>, rot=<-0.0544 0.1502 0.9852 0.062>}
 Dyn_amulet_01 {parent=Dyn_amulet_01_Base, loc=<0 0 0>, rot=<0 0 0 1>}
 Dyn_amulet_02 {parent=Dyn_amulet_01, loc=<2.3052 -0.0412 -0.0263>, rot=<-0.0159 0.0077 -0.0285 -0.9994>}
 Dyn_amulet_10 {parent=Dyn_amulet_02, loc=<6.803 -0.2227 2.4454>, rot=<0.0248 0.0475 -0.0037 -0.9986>}
 Dyn_amulet_11 {parent=Dyn_amulet_02, loc=<8.2732 -0.2325 -0.0226>, rot=<0.0173 0.0172 0.0079 -0.9997>}
 Dyn_amulet_12 {parent=Dyn_amulet_02, loc=<6.7273 -0.2304 -2.0862>, rot=<-0.0016 -0.1145 0.0139 -0.9933>}
 root {parent=root, type=Socket, loc=<0 0 0>, rot=<0 0 0 1>}
]
