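# Physics asset for the Fly mesh of the female medium PaperGlow player body, setting up skirt and belt physics.
# Capsules on pelvis, lower spine, legs and feet, plus a large spine_03 box, form the body, joined by free constraints.
# Skirt chains hang from pelvis and belt chains from spine_01, with swing limits tightening down each chain and zero twist.
# Front and back belt tips are constrained together, closing a loop.
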PhysicsAsset /BRCosmetics/Characters/Player/Female/Medium/Bodies/F_MED_PaperGlow/Meshes/F_MED_PaperGlow_Fly_Physics
bodies {
 Dyn_belt_Front_01 [{
   shapeType=Capsule
   length=4.7661
   radius=4.4173
   shapeName=Dyn_belt_Front_01_capsule
   center=<6.4392 -1.7622 -0.4861>
   rotation=<rot 90 86.4713 81.4715>
  }]
 Dyn_belt_Front_02 [{
   shapeType=Capsule
   length=2.6575
   radius=6.3753
   shapeName=Dyn_belt_Front_02_capsule
   center=<5.4 -4.056 -1.2779>
   rotation=<rot -9.9998 -160.0001 174.9999>
  }]
 Dyn_belt_Front_03 [{shapeType=Capsule, length=2.9493, radius=1.8141, shapeName=Dyn_belt_Front_03_capsule, center=<1.9355 0 0>, rotation=<rot 0 180 179.9998>}]
 Dyn_belt_Front_Base [{shapeType=Capsule, length=0.505, radius=0.505, shapeName=Dyn_belt_Front_Base_capsule}]
 Dyn_belt_back_01 [{
   shapeType=Capsule
   length=6.9831
   radius=4.0053
   shapeName=Dyn_belt_back_01_capsule
   center=<5.8522 1.3537 -0.6046>
   rotation=<rot 85 174.9971 -175.001>
  }]
 Dyn_belt_back_02 [{
   shapeType=Capsule
   length=2.9962
   radius=6.2142
   shapeName=Dyn_belt_back_02_capsule
   center=<9.6283 -0.3965 0.2544>
   rotation=<rot 20.0007 -35.0002 -9.9996>
  }]
 Dyn_belt_back_03 [{shapeType=Capsule, length=2.9493, radius=1.8141, shapeName=Dyn_belt_back_03_capsule, center=<1.9355 0 0>, rotation=<rot 0 180 179.9998>}]
 Dyn_belt_back_Base [{shapeType=Capsule, length=0.505, radius=0.505, shapeName=Dyn_belt_back_Base_capsule}]
 Dyn_skirt_Front_01 [{shapeType=Capsule, length=10.5655, radius=5.6444, shapeName=Dyn_skirt_Front_01_capsule, center=<11.2021 0 1.6282>, rotation=<rot 90 0 0>}]
 Dyn_skirt_Front_01_Base [{shapeType=Capsule, length=0.505, radius=0.505, shapeName=Dyn_skirt_Front_01_Base_capsule, rotation=<rot 0 -90 0>}]
 Dyn_skirt_Front_02 [{shapeType=Capsule, length=9.6559, radius=6.3756, shapeName=Dyn_skirt_Front_02_capsule, center=<11.2021 0 1.6282>, rotation=<rot 90 0 0>}]
 Dyn_skirt_L_01 [{
   shapeType=Capsule
   length=14.5166
   radius=6.4834
   shapeName=Dyn_skirt_L_01_capsule
   center=<7.9024 0.1884 3.8271>
   rotation=<rot 0 0.0001 95>
  }]
 Dyn_skirt_L_01_Base [{shapeType=Capsule, length=0.505, radius=0.505, shapeName=Dyn_skirt_L_01_Base_capsule}]
 Dyn_skirt_L_02 [{
   shapeType=Capsule
   length=13.5892
   radius=7.4108
   shapeName=Dyn_skirt_L_02_capsule
   center=<6.7633 0.2291 4.2929>
   rotation=<rot 0 0.0001 95>
  }]
 Dyn_skirt_L_03 [{
   shapeType=Capsule
   length=13.3829
   radius=7.6171
   shapeName=Dyn_skirt_L_03_capsule
   center=<6.7633 0.1946 3.8983>
   rotation=<rot 0 0.0001 95>
  }]
 Dyn_skirt_L_04 [{shapeType=Capsule, length=13, radius=8, shapeName=Dyn_skirt_L_04_capsule, center=<6.7633 0.1946 3.8983>, rotation=<rot 0 0.0001 95>}]
 Dyn_skirt_L_06 [{shapeType=Capsule, length=15.4057, radius=5.5943, shapeName=Dyn_skirt_L_06_capsule, center=<5.382 0 3.432>, rotation=<rot 0 0 90>}]
 Dyn_skirt_L_06_Base [{shapeType=Capsule, length=0.505, radius=0.505, shapeName=Dyn_skirt_L_06_Base_capsule}]
 Dyn_skirt_L_07 [{shapeType=Capsule, length=14.1788, radius=6.8212, shapeName=Dyn_skirt_L_07_capsule, center=<5.382 0 3.432>, rotation=<rot 0 0 90>}]
 Dyn_skirt_L_08 [{shapeType=Capsule, length=13.4684, radius=7.5316, shapeName=Dyn_skirt_L_08_capsule, center=<5.382 0 4.7549>, rotation=<rot 0 0 90>}]
 Dyn_skirt_L_09 [{shapeType=Capsule, length=13, radius=8, shapeName=Dyn_skirt_L_09_capsule, center=<5.382 0 4.7549>, rotation=<rot 0 0 90>}]
 Dyn_skirt_R_01_Base [{shapeType=Capsule, length=0.4613, radius=2.9263, shapeName=Dyn_skirt_R_01_Base_capsule}]
 Dyn_skirt_R_02 [{
   shapeType=Capsule
   length=8.5544
   radius=6.4456
   shapeName=Dyn_skirt_R_02_capsule
   center=<7.2692 -0.1467 3.5379>
   rotation=<rot -0.0001 -0.0003 -89.9999>
  }]
 Dyn_skirt_R_03 [{
   shapeType=Capsule
   length=8.2643
   radius=6.7357
   shapeName=Dyn_skirt_R_03_capsule
   center=<6.6104 -0.1466 3.4331>
   rotation=<rot -0.0001 -0.0003 -89.9999>
  }]
 Dyn_skirt_R_04 [{
   shapeType=Capsule
   length=9.0854
   radius=7.1846
   shapeName=Dyn_skirt_R_04_capsule
   center=<6.6104 -0.1466 2.6305>
   rotation=<rot -0.0001 -0.0003 -90>
  }]
 Dyn_skirt_R_05 [{
   shapeType=Capsule
   length=11.3733
   radius=8.0627
   shapeName=Dyn_skirt_R_05_capsule
   center=<6.6104 -0.1466 2.1483>
   rotation=<rot -0.0001 -0.0003 -85>
  }]
 Dyn_skirt_back_01 [{shapeType=Capsule, length=15.2061, radius=5.7939, shapeName=Dyn_skirt_back_01_capsule, center=<5.3752 0 1.8117>, rotation=<rot 0 0 90>}]
 Dyn_skirt_back_01_Base [{shapeType=Capsule, length=0.505, radius=0.505, shapeName=Dyn_skirt_back_01_Base_capsule}]
 Dyn_skirt_back_02 [{shapeType=Capsule, length=13.5861, radius=7.4139, shapeName=Dyn_skirt_back_02_capsule, center=<5.3752 0 1.8253>, rotation=<rot 0 0 90>}]
 Dyn_skirt_back_03 [{shapeType=Capsule, length=16.2269, radius=8.5521, shapeName=Dyn_skirt_back_03_capsule, center=<6.1061 0 2.4208>, rotation=<rot 0 0 90>}]
 Dyn_skirt_back_04 [{shapeType=Capsule, length=24.096, radius=9.7402, shapeName=Dyn_skirt_back_04_capsule, center=<7.1302 0 2.703>, rotation=<rot 0 0 90>}]
 calf_l [{shapeType=Capsule, length=27.3376, radius=9.7718, shapeName=calf_l_capsule, center=<-19.8234 -1.893 0>, rotation=<rot 90 90 85.0001>}]
 calf_r [{shapeType=Capsule, length=25.4119, radius=9.779, shapeName=calf_r_capsule, center=<20.5763 1.8002 0>, rotation=<rot 90 80.7735 80.7736>}]
 foot_l [{shapeType=Capsule, length=10.3973, radius=9.5206, shapeName=foot_r_capsule, center=<-2.0599 5.1673 0>, rotation=<rot 0 15 -90>}]
 foot_r [{shapeType=Capsule, length=10.3973, radius=9.0659, shapeName=foot_r_capsule, center=<2.0599 -5.1673 0>, rotation=<rot 0 15 90>}]
 pelvis [{shapeType=Capsule, length=7.0447, radius=10.0127, shapeName=pelvis_capsule, center=<0 -0.0597 0.08>, rotation=<rot 0 -0.0001 5>}]
 spine_01 [{shapeType=Capsule, length=5.6761, radius=8.4438, shapeName=spine_01_capsule, center=<0 0.465 0>}]
 spine_02 [{shapeType=Capsule, length=3.9279, radius=9.5221, shapeName=spine_02_capsule, center=<0 1.568 0>}]
 spine_03 [{
   shapeType=Box
   extentX=6.8237
   extentY=67.0649
   extentZ=71.8184
   shapeName=spine_03_box
   center=<1.7397 -2.3662 0>
   rotation=<rot 0 24.9999 0>
  }]
 thigh_l [{shapeType=Capsule, length=25.4166, radius=9.3385, shapeName=thigh_l_capsule, center=<-18.8187 -0.5194 0.4625>, rotation=<rot 90 0 0>}]
 thigh_r [{shapeType=Capsule, length=23.3748, radius=8.6611, shapeName=thigh_r_capsule, center=<22.0662 1.1502 0>, rotation=<rot 90 0 0>}]
}
constraints [
 {
  bone1Name=spine_01
  bone2Name=pelvis
  swing1LimitDegrees=45
  swing1Motion=Free
  swing2LimitDegrees=45
  swing2Motion=Free
  twistLimitDegrees=45
  twistMotion=Free
 }
 {
  bone1Name=spine_02
  bone2Name=spine_01
  swing1LimitDegrees=45
  swing1Motion=Free
  swing2LimitDegrees=45
  swing2Motion=Free
  twistLimitDegrees=45
  twistMotion=Free
 }
 {
  bone1Name=thigh_l
  bone2Name=pelvis
  swing1LimitDegrees=45
  swing1Motion=Free
  swing2LimitDegrees=45
  swing2Motion=Free
  twistLimitDegrees=45
  twistMotion=Free
 }
 {
  bone1Name=calf_l
  bone2Name=thigh_l
  swing1LimitDegrees=45
  swing1Motion=Free
  swing2LimitDegrees=45
  swing2Motion=Free
  twistLimitDegrees=45
  twistMotion=Free
 }
 {
  bone1Name=thigh_r
  bone2Name=pelvis
  swing1LimitDegrees=45
  swing1Motion=Free
  swing2LimitDegrees=45
  swing2Motion=Free
  twistLimitDegrees=45
  twistMotion=Free
 }
 {
  bone1Name=calf_r
  bone2Name=thigh_r
  swing1LimitDegrees=45
  swing1Motion=Free
  swing2LimitDegrees=45
  swing2Motion=Free
  twistLimitDegrees=45
  twistMotion=Free
 }
 {
  bone1Name=foot_l
  bone2Name=calf_l
  swing1LimitDegrees=45
  swing1Motion=Free
  swing2LimitDegrees=45
  swing2Motion=Free
  twistLimitDegrees=45
  twistMotion=Free
 }
 {
  bone1Name=foot_r
  bone2Name=calf_r
  swing1LimitDegrees=15
  swing1Motion=Free
  swing2LimitDegrees=15
  swing2Motion=Free
  twistLimitDegrees=0
  twistMotion=Free
 }
 {bone1Name=Dyn_skirt_L_01, bone2Name=Dyn_skirt_L_01_Base, swing1LimitDegrees=10, swing2LimitDegrees=10, twistLimitDegrees=0}
 {bone1Name=Dyn_skirt_L_02, bone2Name=Dyn_skirt_L_01, swing1LimitDegrees=8, swing2LimitDegrees=8, twistLimitDegrees=0}
 {bone1Name=Dyn_skirt_L_03, bone2Name=Dyn_skirt_L_02, swing1LimitDegrees=6, swing2LimitDegrees=6, twistLimitDegrees=0}
 {bone1Name=Dyn_skirt_L_04, bone2Name=Dyn_skirt_L_03, swing1LimitDegrees=4, swing2LimitDegrees=4, twistLimitDegrees=0}
 {bone1Name=Dyn_skirt_L_06, bone2Name=Dyn_skirt_L_06_Base, swing1LimitDegrees=10, swing2LimitDegrees=10, twistLimitDegrees=0}
 {bone1Name=Dyn_skirt_L_07, bone2Name=Dyn_skirt_L_06, swing1LimitDegrees=8, swing2LimitDegrees=8, twistLimitDegrees=0}
 {bone1Name=Dyn_skirt_L_08, bone2Name=Dyn_skirt_L_07, swing1LimitDegrees=6, swing2LimitDegrees=6, twistLimitDegrees=0}
 {bone1Name=Dyn_skirt_L_09, bone2Name=Dyn_skirt_L_08, swing1LimitDegrees=4, swing2LimitDegrees=4, twistLimitDegrees=0}
 {bone1Name=Dyn_skirt_back_01, bone2Name=Dyn_skirt_back_01_Base, swing1LimitDegrees=10, swing2LimitDegrees=10, twistLimitDegrees=0}
 {bone1Name=Dyn_skirt_back_02, bone2Name=Dyn_skirt_back_01, swing1LimitDegrees=8, swing2LimitDegrees=8, twistLimitDegrees=0}
 {bone1Name=Dyn_skirt_back_03, bone2Name=Dyn_skirt_back_02, swing1LimitDegrees=6, swing2LimitDegrees=6, twistLimitDegrees=0}
 {bone1Name=Dyn_skirt_back_04, bone2Name=Dyn_skirt_back_03, swing1LimitDegrees=4, swing2LimitDegrees=4, twistLimitDegrees=0}
 {bone1Name=Dyn_skirt_R_02, bone2Name=Dyn_skirt_R_01_Base, swing1LimitDegrees=10, swing2LimitDegrees=10, twistLimitDegrees=0}
 {bone1Name=Dyn_skirt_R_03, bone2Name=Dyn_skirt_R_02, swing1LimitDegrees=8, swing2LimitDegrees=8, twistLimitDegrees=0}
 {bone1Name=Dyn_skirt_R_04, bone2Name=Dyn_skirt_R_03, swing1LimitDegrees=6, swing2LimitDegrees=6, twistLimitDegrees=0}
 {bone1Name=Dyn_skirt_R_05, bone2Name=Dyn_skirt_R_04, swing1LimitDegrees=4, swing2LimitDegrees=4, twistLimitDegrees=0}
 {bone1Name=Dyn_belt_back_02, bone2Name=Dyn_belt_back_01, swing1LimitDegrees=15, swing2LimitDegrees=18, twistLimitDegrees=0}
 {bone1Name=Dyn_belt_back_03, bone2Name=Dyn_belt_back_02, swing1LimitDegrees=10, swing2LimitDegrees=10, twistLimitDegrees=0}
 {bone1Name=Dyn_belt_Front_01, bone2Name=Dyn_belt_Front_Base, swing1LimitDegrees=20, swing2LimitDegrees=25, twistLimitDegrees=0}
 {bone1Name=Dyn_belt_Front_02, bone2Name=Dyn_belt_Front_01, swing1LimitDegrees=15, swing2LimitDegrees=22, twistLimitDegrees=0}
 {bone1Name=Dyn_belt_Front_03, bone2Name=Dyn_belt_Front_02, swing1LimitDegrees=5, swing2LimitDegrees=5, twistLimitDegrees=0}
 {bone1Name=Dyn_belt_Front_03, bone2Name=Dyn_belt_back_03, swing1LimitDegrees=10, swing2LimitDegrees=10, twistLimitDegrees=0}
 {
  bone1Name=Dyn_skirt_R_01_Base
  bone2Name=pelvis
  swing1LimitDegrees=45
  swing1Motion=Free
  swing2LimitDegrees=45
  swing2Motion=Free
  twistLimitDegrees=45
  twistMotion=Free
 }
 {
  bone1Name=Dyn_skirt_L_01_Base
  bone2Name=pelvis
  swing1LimitDegrees=45
  swing1Motion=Free
  swing2LimitDegrees=45
  swing2Motion=Free
  twistLimitDegrees=45
  twistMotion=Free
 }
 {
  bone1Name=Dyn_skirt_L_06_Base
  bone2Name=pelvis
  swing1LimitDegrees=45
  swing1Motion=Free
  swing2LimitDegrees=45
  swing2Motion=Free
  twistLimitDegrees=45
  twistMotion=Free
 }
 {
  bone1Name=Dyn_skirt_back_01_Base
  bone2Name=pelvis
  swing1LimitDegrees=45
  swing1Motion=Free
  swing2LimitDegrees=45
  swing2Motion=Free
  twistLimitDegrees=45
  twistMotion=Free
 }
 {
  bone1Name=Dyn_belt_Front_Base
  bone2Name=spine_01
  swing1LimitDegrees=45
  swing1Motion=Free
  swing2LimitDegrees=45
  swing2Motion=Free
  twistLimitDegrees=45
  twistMotion=Free
 }
 {
  bone1Name=Dyn_belt_back_Base
  bone2Name=spine_01
  swing1LimitDegrees=45
  swing1Motion=Free
  swing2LimitDegrees=45
  swing2Motion=Free
  twistLimitDegrees=45
  twistMotion=Free
 }
 {
  bone1Name=Dyn_skirt_Front_01_Base
  bone2Name=pelvis
  swing1LimitDegrees=45
  swing1Motion=Free
  swing2LimitDegrees=45
  swing2Motion=Free
  twistLimitDegrees=45
  twistMotion=Free
 }
 {bone1Name=Dyn_skirt_Front_01, bone2Name=Dyn_skirt_Front_01_Base, swing1LimitDegrees=20, swing2LimitDegrees=25, twistLimitDegrees=0}
 {bone1Name=Dyn_skirt_Front_02, bone2Name=Dyn_skirt_Front_01, swing1LimitDegrees=15, swing2LimitDegrees=15, twistLimitDegrees=0}
 {bone1Name=Dyn_belt_back_01, bone2Name=Dyn_belt_back_Base, swing1LimitDegrees=20, swing2LimitDegrees=23, twistLimitDegrees=0}
 {
  bone1Name=spine_03
  bone2Name=spine_02
  swing1LimitDegrees=45
  swing1Motion=Free
  swing2LimitDegrees=45
  swing2Motion=Free
  twistLimitDegrees=45
  twistMotion=Free
 }
]
