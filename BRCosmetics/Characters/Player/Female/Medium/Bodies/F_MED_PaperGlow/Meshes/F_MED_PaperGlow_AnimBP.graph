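# Anim Blueprint for the F_MED_PaperGlow character body, adding secondary motion to skirt, belt, cape, pendant and bracelets on the pose copied from the parent mesh.
# Deform and state-driven Control Rigs feed RigidBody simulation, switching to flight physics with dynamic gravity while skydiving, then further RigidBody and AnimDynamics passes.
# An ALT Control Rig takes over when rigid bodies are disabled.
# The EventGraph updates gravity, weapon, LOD, emote and cape-backbling inputs.
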AnimBlueprint /BRCosmetics/Characters/Player/Female/Medium/Bodies/F_MED_PaperGlow/Meshes/F_MED_PaperGlow_AnimBP
target_skeleton=&/F_MED_PaperGlow_Skeleton
graphs [
 AnimGraph {nodes [
   Anim Dynamics - Bone: Dyn_pendant_02(Alpha=1, ExternalForce="0,0,0") {
    angularDampingOverride=0.999
    angularSpringConstant=300
    bAngularSpring=true
    bOverrideAngularDamping=true
    bUseGravityOverride=true
    bUsePlanarLimit=false
    simulationSpace=World
    alphaBoolBlend {blendOption=Linear}
    boundBone {boneName=Dyn_pendant_02}
    componentAppliedLinearAccClamp=<10000 10000 10000>
    gravityOverride {z=-980}
    physicsBodyDefinitions [{
      boundBone {boneName=Dyn_pendant_02}
      boxExtents=<5 5 5>
      constraintSetup {angularLimitsMax=<5 40 20>, angularLimitsMin=<-5 -40 -5>, angularTarget=<0 0 0>}
      localJointOffset=<0.3 0 0>
     }]
   }
   "Transform (Modify) Bone - Bone: Dyn_cape_back"(Alpha=1, Translation="0,-3,0", Rotation="0,0,5", Scale="1,1,1") {
    rotationMode=BMM_Additive
    scaleMode=BMM_Additive
    translationMode=BMM_Additive
    alphaBoolBlend {blendOption=Linear}
    boneToModify {boneName=Dyn_cape_back}
    rotation {roll=5}
    translation {y=-3}
   }
   n1 = Control Rig!(Source=Use cached pose 'pre sim'(), bAlphaBoolEnabled:bool=True) -> (Pose) {
    alphaInputType=Bool
    alphaBoolBlend {blendOption=Linear}
    controlRigAssetReference {blueprintRigClass=&/F_MED_PaperGlow_Deform_CtrlRig.F_MED_PaperGlow_Deform_CtrlRig_C}
   }
   n2 = Control Rig!(
    Source=@n1
    bAlphaBoolEnabled=Get bRigidBodyEnabled()
    IsKneeSliding:bool=Get bIsKneeSliding()
    iscrouch:bool=Get bIsCrouching()
    IsTacticalSprinting:bool=Get bIsTacticalSprinting()
    IsBaseballSliding:bool=Get bIsBaseballSliding()
    IsMoving2D:bool=Get bIsMoving2D()
    IsOnGrindRail:bool=Get bIsOnGrindRail()
    IsPlayingEmote:bool=Get bIsPlayingEmote()
    IsTargeting:bool=Get bIsTargeting()
    isPistol:bool=Get isPistol() -> (isPistol:bool)
    isMelee:bool=Get isMelee() -> (isMelee:bool)
    IsSkyDiving:bool=Get bIsSkyDiving()
    IsJumping:bool=Get bIsJumping()
    IsParachuteOpened:bool=Get bIsParachuteOpened()
    DBNO:bool=Get bIsDBNO()
    IsInFrontEnd:bool=Get bIsInFrontEnd()
    Wear_cape:bool=Get Wear_cape() -> (Wear_cape:bool)
   ) -> (Pose) {
    alphaInputType=Bool
    alphaBoolBlend {blendOption=Linear}
    controlRigAssetReference {blueprintRigClass=&/F_MED_PaperGlow_CtrlRig.F_MED_PaperGlow_CtrlRig_C}
   }
   n3 = "Transform (Modify) Bone - Bone: Dyn_cape_back"(ComponentPose=Local To Component(LocalPose=@n2), Alpha=1, Translation="0,1,0", Rotation="0.000,0.000,5.000", Scale="1,1,1") {
    rotationMode=BMM_Additive
    scaleMode=BMM_Additive
    translationMode=BMM_Additive
    alphaBoolBlend {blendOption=Linear}
    boneToModify {boneName=Dyn_cape_back}
    rotation {roll=5}
    translation {y=1}
   }
   n4 = "Transform (Modify) Bone - Bone: Dyn_belt_Front_Base"(ComponentPose=@n3, Alpha=1, Translation="0,3,0", Rotation="-0.000,-10.000,0.000", Scale="1,1,1") {
    rotationMode=BMM_Additive
    scaleMode=BMM_Additive
    translationMode=BMM_Additive
    alphaBoolBlend {blendOption=Linear}
    boneToModify {boneName=Dyn_belt_Front_Base}
    rotation {yaw=-10}
    translation {y=3}
   }
   n5 = "Transform (Modify) Bone - Bone: Dyn_belt_back_Base"(ComponentPose=@n4, Alpha=1, Translation="0,0,0", Rotation="0.000,10.000,-0.000", Scale="1,1,1") {
    rotationMode=BMM_Additive
    scaleMode=BMM_Additive
    translationMode=BMM_Additive
    alphaBoolBlend {blendOption=Linear}
    boneToModify {boneName=Dyn_belt_back_Base}
    rotation {yaw=10}
   }
   n6 = "Transform (Modify) Bone - Bone: Dyn_skirt_R_01_Base"!(ComponentPose=@n5, bAlphaBoolEnabled:bool=Get bRigidBodyEnabled(), Translation="-4,-2,0", Rotation="0,0,0", Scale="1,1,1") -> (Pose) {
    alphaInputType=Bool
    rotationMode=BMM_Additive
    scaleMode=BMM_Additive
    translationMode=BMM_Additive
    alphaBoolBlend {blendOption=Linear}
    boneToModify {boneName=Dyn_skirt_R_01_Base}
    translation=<-4 -2>
   }
   n7 = "Transform (Modify) Bone - Bone: Dyn_skirt_back_01_Base"!(ComponentPose=@n6, bAlphaBoolEnabled:bool=Get bRigidBodyEnabled(), Translation="0,-1,0", Rotation="0,0,0", Scale="1,1,1") -> (Pose) {
    alphaInputType=Bool
    rotationMode=BMM_Additive
    scaleMode=BMM_Additive
    translationMode=BMM_Additive
    alphaBoolBlend {blendOption=Linear}
    boneToModify {boneName=Dyn_skirt_back_01_Base}
    translation {y=-1}
   }
   n8 = RigidBody!(ComponentPose=@n7, bAlphaBoolEnabled:bool, OverrideWorldGravity:Vector="0,0,-980", ExternalForce) -> (Pose) {
    alphaInputType=Bool
    bClampLinearTranslationLimitToRefPose=true
    bOverrideWorldGravity=true
    overridePhysicsAsset=&/F_MED_PaperGlow_Physics
    simulationSpace=BaseBoneSpace
    alphaBoolBlend {blendOption=Linear}
    alphaScaleBiasClamp {bInterpResult=true, interpSpeedDecreasing=3, interpSpeedIncreasing=3}
    baseBoneRef {boneName=pelvis}
    overrideWorldGravity {z=-980}
    simSpaceSettings {
     maxAngularAcceleration=300
     maxAngularVelocity=1
     maxLinearAcceleration=300
     maxLinearVelocity=150
     velocityScaleZ=0
     worldAlpha=1
     externalLinearDragV=<0.1 0.1>
    }
   }
   n9 = Component To Local(ComponentPose=@n8)
   n10 = Control Rig!(Source=Use cached pose 'pre sim'(), bAlphaBoolEnabled=True) -> (Pose) {
    alphaInputType=Bool
    alphaBoolBlend {blendOption=Linear}
    controlRigAssetReference {blueprintRigClass=&F_MED_PaperGlow_Deform_CtrlRig_C}
   }
   n11 = Control Rig!(
    Source=@n10
    bAlphaBoolEnabled=Get bRigidBodyEnabled()
    IsKneeSliding:bool=Get bIsKneeSliding()
    iscrouch:bool=Get bIsCrouching()
    IsTacticalSprinting:bool=Get bIsTacticalSprinting()
    IsBaseballSliding:bool=Get bIsBaseballSliding()
    IsMoving2D:bool=Get bIsMoving2D()
    IsOnGrindRail:bool=Get bIsOnGrindRail()
    IsPlayingEmote:bool=Get bIsPlayingEmote()
    IsTargeting:bool=Get bIsTargeting()
    isPistol:bool=Get isPistol()
    isMelee:bool=Get isMelee()
    IsSkyDiving:bool=Get bIsSkyDiving()
    IsJumping:bool=Get bIsJumping()
    IsParachuteOpened:bool=Get bIsParachuteOpened()
    DBNO:bool=Get bIsDBNO()
    IsInFrontEnd:bool=Get bIsInFrontEnd()
    Wear_cape:bool=Get Wear_cape()
   ) -> (Pose) {
    alphaInputType=Bool
    alphaBoolBlend {blendOption=Linear}
    controlRigAssetReference {blueprintRigClass=&F_MED_PaperGlow_CtrlRig_C}
   }
   n12 = "Transform (Modify) Bone - Bone: Dyn_belt_Front_Base"(ComponentPose=Local To Component(LocalPose=@n11), Alpha=1, Translation="0,3,0", Rotation="-0.000,-10.000,0.000", Scale="1,1,1") {
    rotationMode=BMM_Additive
    scaleMode=BMM_Additive
    translationMode=BMM_Additive
    alphaBoolBlend {blendOption=Linear}
    boneToModify {boneName=Dyn_belt_Front_Base}
    rotation {yaw=-10}
    translation {y=3}
   }
   n13 = "Transform (Modify) Bone - Bone: Dyn_belt_back_Base"(ComponentPose=@n12, Alpha=1, Translation="0,0,0", Rotation="0.000,10.000,-0.000", Scale="1,1,1") {
    rotationMode=BMM_Additive
    scaleMode=BMM_Additive
    translationMode=BMM_Additive
    alphaBoolBlend {blendOption=Linear}
    boneToModify {boneName=Dyn_belt_back_Base}
    rotation {yaw=10}
   }
   n14 = "Transform (Modify) Bone - Bone: Dyn_skirt_R_01_Base"(ComponentPose=@n13, bAlphaBoolEnabled=Get bRigidBodyEnabled(), Translation="-4,-2,0", Rotation="0,0,0", Scale="1,1,1") {
    alphaInputType=Bool
    rotationMode=BMM_Additive
    scaleMode=BMM_Additive
    translationMode=BMM_Additive
    alphaBoolBlend {blendOption=Linear}
    boneToModify {boneName=Dyn_skirt_R_01_Base}
    translation=<-4 -2>
   }
   n15 = "Transform (Modify) Bone - Bone: Dyn_skirt_back_01_Base"(ComponentPose=@n14, bAlphaBoolEnabled=Get bRigidBodyEnabled(), Translation="0,-1,0", Rotation="0,0,0", Scale="1,1,1") {
    alphaInputType=Bool
    rotationMode=BMM_Additive
    scaleMode=BMM_Additive
    translationMode=BMM_Additive
    alphaBoolBlend {blendOption=Linear}
    boneToModify {boneName=Dyn_skirt_back_01_Base}
    translation {y=-1}
   }
   n16 = RigidBody!(ComponentPose=@n15, Alpha:float, OverrideWorldGravity=Get Gravity Override(), ExternalForce) -> (Pose) {
    bOverrideWorldGravity=true
    overridePhysicsAsset=&/F_MED_PaperGlow_Fly_Physics
    simulationSpace=BaseBoneSpace
    alphaBoolBlend {blendOption=Linear}
    baseBoneRef {boneName=pelvis}
    componentAppliedLinearAccClamp=<3000 3000 3000>
    componentLinearAccScale {z=1}
    componentLinearVelScale {z=1}
    overrideWorldGravity {z=-400}
    simSpaceSettings {
     maxAngularAcceleration=100
     maxAngularVelocity=3
     maxLinearAcceleration=600
     maxLinearVelocity=150
     velocityScaleZ=0
     worldAlpha=1
     externalLinearDragV=<0.1 0.1>
    }
   }
   n17 = Component To Local(ComponentPose=@n16)
   n18 = Get bIsSkyDiving()
   n19 = RigidBody!(
    ComponentPose=Local To Component(LocalPose=Blend Poses by bool(BlendPose_0=@n17, BlendPose_1=@n9, BlendTime_0=0, bActiveValue=@n18) {blendType=Linear})
    bAlphaBoolEnabled:bool=Get bRigidBodyEnabled()
    OverrideWorldGravity="0,0,-980"
    ExternalForce
   ) -> (Pose) {
    alphaInputType=Bool
    bClampLinearTranslationLimitToRefPose=true
    bOverrideWorldGravity=true
    overridePhysicsAsset=&/F_MED_PaperGlowKulon_Physics
    simulationSpace=BaseBoneSpace
    alphaBoolBlend {blendOption=Linear}
    baseBoneRef {boneName=thigh_l}
    overrideWorldGravity {z=-980}
    simSpaceSettings {
     maxAngularAcceleration=400
     maxAngularVelocity=3
     maxLinearAcceleration=300
     maxLinearVelocity=200
     velocityScaleZ=0
     worldAlpha=1
     externalLinearDragV=<0.1 0.1>
    }
   }
   n20 = RigidBody(ComponentPose=@n19, OverrideWorldGravity="0,0,-980") {
    alphaInputType=Bool
    bClampLinearTranslationLimitToRefPose=true
    bOverrideWorldGravity=true
    overridePhysicsAsset=&/F_MED_PaperGlow_Cape_Physics
    simulationSpace=BaseBoneSpace
    alphaBoolBlend {blendOption=Linear}
    baseBoneRef {boneName=spine_04}
    componentAppliedLinearAccClamp=<1000 1000 1000>
    simSpaceSettings {
     maxAngularAcceleration=50
     maxAngularVelocity=3
     maxLinearAcceleration=40
     maxLinearVelocity=25
     velocityScaleZ=0
     worldAlpha=1
     externalLinearDragV=<0.1 0.1>
    }
   }
   n21 = Anim Dynamics - Bone: Dyn_pendant_02(ComponentPose=@n20, Alpha=1, ExternalForce="0,0,0") {
    angularDampingOverride=0.9999
    angularSpringConstant=50
    bAngularSpring=true
    bOverrideAngularDamping=true
    simulationSpace=RootRelative
    alphaBoolBlend {blendOption=Linear}
    boundBone {boneName=Dyn_pendant_02}
    physicsBodyDefinitions [{
      boundBone {boneName=Dyn_pendant_02}
      boxExtents=<7 7 7>
      constraintSetup {angularLimitsMax=<0 20 50>, angularLimitsMin=<0 -20 0>, angularTarget=<0 0 1>}
      localJointOffset=<0.8 0 0>
     }]
   }
   n22 = Anim Dynamics - Bone: Dyn_cape_back(ComponentPose=@n21, Alpha=1, ExternalForce="0,0,0") {
    angularDampingOverride=0.999
    angularSpringConstant=2000
    bAngularSpring=true
    bGravityOverrideInSimSpace=true
    bOverrideAngularDamping=true
    bUseGravityOverride=true
    bUsePlanarLimit=false
    bUseSphericalLimits=true
    simulationSpace=RootRelative
    alphaBoolBlend {blendOption=Linear}
    boundBone {boneName=Dyn_cape_back}
    componentAppliedLinearAccClamp=<2000 2000 2000>
    componentLinearAccScale=<3 3 3>
    componentLinearVelScale=<3 3 3>
    gravityOverride {z=-980}
    physicsBodyDefinitions [{
      boundBone {boneName=Dyn_cape_back}
      boxExtents=<8 8 8>
      constraintSetup {angularTargetAxis=AxisZ, angularLimitsMax=<0 10 5>, angularLimitsMin=<0 -10 -15>, angularTarget=<0 0 1>}
      localJointOffset=<0.2 0 0>
     }]
   }
   n23 = Anim Dynamics - Bone: Dyn_bracelet_L_01(ComponentPose=@n22, Alpha=1, ExternalForce="0,0,0") {
    angularSpringConstant=150
    bAngularSpring=true
    bUseGravityOverride=true
    bUsePlanarLimit=false
    alphaBoolBlend {blendOption=Linear}
    boundBone {boneName=Dyn_bracelet_L_01}
    gravityOverride {z=-980}
    physicsBodyDefinitions [{
      boundBone {boneName=Dyn_bracelet_L_01}
      boxExtents=<10 10 10>
      constraintSetup {angularLimitsMax=<3 0 0>, angularLimitsMin=<-3 -15 0>, angularTarget=<0 0 0>}
      localJointOffset=<0.4 0 0>
     }]
   }
   n24 = Anim Dynamics - Bone: Dyn_bracelet_R(ComponentPose=@n23, Alpha=1, ExternalForce="0,0,0") {
    angularSpringConstant=100
    bAngularSpring=true
    bUseGravityOverride=true
    simulationSpace=World
    alphaBoolBlend {blendOption=Linear}
    boundBone {boneName=Dyn_bracelet_R}
    gravityOverride {z=-980}
    physicsBodyDefinitions [{
      boundBone {boneName=Dyn_bracelet_R}
      boxExtents=<5 5 5>
      constraintSetup {angularLimitsMax=<3 3 3>, angularLimitsMin=<-3 -3 -3>, angularTarget=<0 0 0>}
      localJointOffset=<1 0 0>
     }]
   }
   n25 = Get FE_notEmote() -> (FE_notEmote:bool)
   n26 = "Transform (Modify) Bone - Bone: Dyn_skirt_L_01_Base"!(ComponentPose=@n24, bAlphaBoolEnabled:bool=@n25, Translation="-3,0,0", Rotation="-3,0,0", Scale="1,1,1") -> (Pose) {
    alphaInputType=Bool
    rotationMode=BMM_Additive
    scaleMode=BMM_Additive
    translationMode=BMM_Additive
    alphaBoolBlend {blendOption=Linear}
    boneToModify {boneName=Dyn_skirt_L_01_Base}
    rotation {pitch=-3}
    translation {x=-3}
   }
   n27 = "Transform (Modify) Bone - Bone: Dyn_skirt_L_06_Base"!(ComponentPose=@n26, bAlphaBoolEnabled:bool=@n25, Translation="-2,0,0", Rotation="-2,0,0", Scale="1,1,1") -> (Pose) {
    alphaInputType=Bool
    rotationMode=BMM_Additive
    scaleMode=BMM_Additive
    translationMode=BMM_Additive
    alphaBoolBlend {blendOption=Linear}
    boneToModify {boneName=Dyn_skirt_L_06_Base}
    rotation {pitch=-2}
    translation {x=-2}
   }
   n28 = "Transform (Modify) Bone - Bone: Dyn_skirt_back_01_Base"(
    ComponentPose=@n27
    bAlphaBoolEnabled=Get NOTRigidBodyEnabled() -> (NOTRigidBodyEnabled:bool)
    Translation="0,-5,5"
    Rotation="0.000,0.000,5.000"
    Scale="1,1,1"
   ) {
    alphaInputType=Bool
    rotationMode=BMM_Additive
    translationMode=BMM_Additive
    alphaBoolBlend {blendOption=Linear}
    boneToModify {boneName=Dyn_skirt_back_01_Base}
    rotation {roll=5}
    translation {y=-5, z=5}
   }
   n29 = "Transform (Modify) Bone - Bone: Dyn_skirt_Front_01"!(ComponentPose=@n28, bAlphaBoolEnabled:bool=Get NOTRigidBodyEnabled(), Translation="-15,0,0", Rotation="0,0,0", Scale="1,1,1") -> (Pose) {
    alphaInputType=Bool
    rotationMode=BMM_Additive
    translationMode=BMM_Additive
    alphaBoolBlend {blendOption=Linear}
    boneToModify {boneName=Dyn_skirt_Front_01}
    translation {x=-15}
   }
   n30 = Control Rig!(
    Source=Component To Local(ComponentPose=@n29)
    bAlphaBoolEnabled=Get NOTRigidBodyEnabled()
    IsPlayingEmote=Get bIsPlayingEmote()
    IsTacticalSprinting=Get bIsTacticalSprinting()
    IsMoving2D=Get bIsMoving2D()
    IsTargeting=Get bIsTargeting()
    iscrouch=Get bIsCrouching()
    Is LOD 2:bool=Get Is LOD 2() -> (Is LOD 2:bool)
    Is LOD 1:bool=Get Is LOD 1() -> (Is LOD 1:bool)
   ) -> (Pose) {
    alphaInputType=Bool
    alphaBoolBlend {blendOption=Linear}
    controlRigAssetReference {blueprintRigClass=&/F_MED_PaperGlow_ALT_CtrlRig.F_MED_PaperGlow_ALT_CtrlRig_C}
   }
   Output Pose(Result=@n30) {layerGroup=None}
   Save cached pose 'pre sim'(Pose=Copy Pose From Mesh() {bUseAttachedParent=true}) {cacheName=pre sim}
   Get coat_dyn_alpha() -> (coat_dyn_alpha:double)
   Get bRigidBodyEnabled()
  ]}
 EventGraph {nodes [
   Event BlueprintUpdateAnimation()
   > n1 = Sequence!(execute) -> (then_0, then_1, then_2:Exec, then_3:Exec)
   > n2 = Gravity Override State Function DEPRECATED(CharacterPartAnimInstance=Self-Reference(), GameTimeInSeconds=GetGameTimeInSeconds(), GravityOverrideParams=Get Gravity Override Params())
   > Set Gravity Override(Gravity Override=@n2)
   > Set NOTRigidBodyEnabled(NOTRigidBodyEnabled:bool=NOT Boolean(A=Get bRigidBodyEnabled())) -> (Output_Get:bool)
   n3 = Cast To FortPlayerAnimInstance(execute=@n1.then_1, Object=GetAnimInstance(self=GetOwnerSkeletalMeshComponent()))
   > n4 = Switch on EFortWeaponCoreAnimation!(execute, Selection:EFortWeaponCoreAnimation Enum=Get WeaponCoreAnim(self=@n3)) -> (
    Melee:Exec
    Pistol:Exec
    Shotgun:Exec
    PaperBlueprint:Exec
    Rifle:Exec
    MeleeOneHand:Exec
    MachinePistol:Exec
    RocketLauncher:Exec
    GrenadeLauncher:Exec
    GoingCommando:Exec
    AssaultRifle:Exec
    TacticalShotgun:Exec
    SniperRifle:Exec
    TrapPlacement:Exec
    ShoulderLauncher:Exec
    AbilityDecoTool:Exec
    Crossbow:Exec
    C4:Exec
    RemoteControl:Exec
    DualWield:Exec
    AR_BullPup:Exec
    AR_ForwardGrip:Exec
    MedPackPaddles:Exec
    SMG_P90:Exec
    AR_DrumGun:Exec
    Consumable_Small:Exec
    Consumable_Large:Exec
    Balloon:Exec
    MountedTurret:Exec
    CreativeTool:Exec
    ExplosiveBow:Exec
    AshtonIndigo:Exec
    AshtonChicago:Exec
    MeleeDualWield:Exec
    Unarmed:Exec
   )
   Set isPistol(
    execute=[
     @n4.AssaultRifle
     @n4.GoingCommando
     @n4.GrenadeLauncher
     @n4.MachinePistol
     @n4.MeleeDualWield
     @n4.MeleeOneHand
     @n4.PaperBlueprint
     @n4.Rifle
     @n4.RocketLauncher
     @n4.Shotgun
     @n4.SniperRifle
     @n4.TacticalShotgun
     @n4.Unarmed
    ]
    isPistol=false
   )
   > Set isMelee(isMelee:bool=false) -> (Output_Get:bool)
   Set isMelee(execute=@n4.Melee, isMelee=true)
   > Set isPistol(isPistol=false)
   Set isPistol(execute=@n4.Pistol, isPistol=true)
   > Set isMelee(isMelee=false)
   n5 = Get bIsKneeSliding()
   n6 = Get bIsBaseballSliding()
   Set coat_dyn_alpha(execute=@n1.then_2, coat_dyn_alpha:double=SelectFloat(A=0.5, B=1, bPickA=OR Boolean(A=@n5, B=@n6))) -> (Output_Get:double)
   > Set Is LOD 2(Is LOD 2:bool="integer >= integer"(A=Get CurrentLOD(), B=2)) -> (Output_Get:bool)
   > Set Is LOD 1(Is LOD 1:bool="Equal (Integer)"(A=Get CurrentLOD(), B=1)) -> (Output_Get:bool)
   n7 = Get bIsPlayingEmote()
   > Set FE_notEmote(FE_notEmote:bool=AND Boolean(A=NOT Boolean(A=@n7), B=Get bIsInFrontEnd())) -> (Output_Get:bool)
   Set Wear_cape(execute=@n1.then_3, Wear_cape:bool=CheckForBackbling(self=Self-Reference()).bIsBackblingACape) -> (Output_Get:bool)
   PrintString!(execute, InString=@n2.CurrentStateOut, bPrintToScreen, bPrintToLog, TextColor, Duration, Key) -> (then)
  ]}
]
variables [
 coat_dyn_alpha {type=number}
 FE_notEmote {type=boolean}
 Gravity Override {type=Vector}
 Gravity Override Params {
  type=GravityOverrideParamsStruct
  default {
   skydive_FloatingDown {gravityOverride=<0 -700 50>, windAmplitude=<200 -500 100>, windFrequency=<1.5 7 1>, windOffset=<0 -600 0>}
   skydive_GoingUp {gravityOverride=<0 -300 50>, windAmplitude=<200 -500 100>, windFrequency=<1.5 7 1>}
  }
 }
 Is LOD 1 {type=boolean}
 Is LOD 2 {type=boolean}
 isMelee {type=boolean}
 isPistol {type=boolean}
 NOTRigidBodyEnabled {type=boolean}
 Wear_cape {type=boolean}
]
