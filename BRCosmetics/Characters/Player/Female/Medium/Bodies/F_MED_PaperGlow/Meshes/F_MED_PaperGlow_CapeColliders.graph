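# Physics asset for the F_MED_PaperGlow female medium character body, providing capsule colliders that, by its name, serve the cape.
# Capsules cover the pelvis, spine, clavicles, head and feet, plus dynamic skirt bone chains at back, left and right and one back cape bone.
# Arm, thigh, calf and neck bodies carry no shapes.
# Constraints follow the bone hierarchy with all swing and twist axes free, so the angular limit values are inactive.
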PhysicsAsset /BRCosmetics/Characters/Player/Female/Medium/Bodies/F_MED_PaperGlow/Meshes/F_MED_PaperGlow_CapeColliders
bodies {
 Dyn_cape_back [{
   shapeType=Capsule
   length=11.7132
   radius=4.4457
   shapeName=Dyn_cape_back_capsule
   center=<10.4765 0 -0.9166>
   rotation=<rot 85 -0.0002 -5.0004>
  }]
 Dyn_skirt_L_01 [{
   shapeType=Capsule
   length=14.7851
   radius=9.2146
   shapeName=Dyn_skirt_L_01_capsule
   center=<5.5715 -0.2017 1.4732>
   rotation=<rot 0 0.0003 84.9999>
  }]
 Dyn_skirt_L_02 [{
   shapeType=Capsule
   length=14.7851
   radius=9.2146
   shapeName=Dyn_skirt_L_02_capsule
   center=<5.5715 -0.2017 1.4732>
   rotation=<rot 0 0.0003 84.9999>
  }]
 Dyn_skirt_L_03 [{
   shapeType=Capsule
   length=14.7851
   radius=9.2146
   shapeName=Dyn_skirt_L_03_capsule
   center=<5.5715 -0.2017 1.4732>
   rotation=<rot 0 0.0003 85>
  }]
 Dyn_skirt_L_04 [{
   shapeType=Capsule
   length=14.7851
   radius=9.2146
   shapeName=Dyn_skirt_L_04_capsule
   center=<5.5715 -0.2017 1.4732>
   rotation=<rot 0 0.0003 85>
  }]
 Dyn_skirt_L_06 [{shapeType=Capsule, length=14.7113, radius=10.1194, shapeName=Dyn_skirt_L_06_capsule, center=<7.9272 0 -0.2698>, rotation=<rot 0 0 90>}]
 Dyn_skirt_L_07 [{shapeType=Capsule, length=14.7113, radius=10.1194, shapeName=Dyn_skirt_L_07_capsule, center=<7.9272 0 -0.2698>, rotation=<rot 0 0 90>}]
 Dyn_skirt_L_08 [{shapeType=Capsule, length=14.7113, radius=10.1194, shapeName=Dyn_skirt_L_08_capsule, center=<7.9272 0 -0.2698>, rotation=<rot 0 0 90>}]
 Dyn_skirt_L_09 [{shapeType=Capsule, length=14.7113, radius=10.1194, shapeName=Dyn_skirt_L_09_capsule, center=<7.9272 0 -0.2698>, rotation=<rot 0 0 90>}]
 Dyn_skirt_R_02 [{
   shapeType=Capsule
   length=18.8943
   radius=9.2233
   shapeName=Dyn_skirt_R_02_capsule
   center=<6.5542 0.2017 0.3029>
   rotation=<rot 0 -0.0001 95>
  }]
 Dyn_skirt_R_03 [{
   shapeType=Capsule
   length=18.8943
   radius=9.2233
   shapeName=Dyn_skirt_R_03_capsule
   center=<6.5542 0.2017 0.3029>
   rotation=<rot 0 -0.0001 95>
  }]
 Dyn_skirt_R_04 [{
   shapeType=Capsule
   length=18.8943
   radius=9.2233
   shapeName=Dyn_skirt_R_04_capsule
   center=<6.5542 0.2017 0.3029>
   rotation=<rot 0 -0.0001 95>
  }]
 Dyn_skirt_R_05 [{
   shapeType=Capsule
   length=18.8943
   radius=9.2233
   shapeName=Dyn_skirt_R_05_capsule
   center=<6.5542 0.2017 0.3029>
   rotation=<rot 0 -0.0001 95>
  }]
 Dyn_skirt_back_01 [{shapeType=Capsule, length=27.7667, radius=8.0068, shapeName=Dyn_skirt_back_01_capsule, center=<8.2454 0 1.7964>, rotation=<rot 0 0 90>}]
 Dyn_skirt_back_02 [{shapeType=Capsule, length=27.2335, radius=8.5401, shapeName=Dyn_skirt_back_02_capsule, center=<8.2454 0 1.7964>, rotation=<rot 0 0 90>}]
 Dyn_skirt_back_03 [{shapeType=Capsule, length=22.2838, radius=13.4898, shapeName=Dyn_skirt_back_03_capsule, center=<12.471 0 -3.7368>, rotation=<rot 0 0 90>}]
 Dyn_skirt_back_04 [{shapeType=Capsule, length=22.8521, radius=12.9215, shapeName=Dyn_skirt_back_04_capsule, center=<12.471 0 -3.7368>, rotation=<rot 0 0 90>}]
 calf_l []
 calf_r []
 clavicle_l [{shapeType=Capsule, length=6.5623, radius=7.5194, center=<3.9566 -1.0862 -2.1127>, rotation=<rot -77.5566 177.3915 7.7314>}]
 clavicle_r [{shapeType=Capsule, length=6.5623, radius=7.5194, center=<-3.9566 1.0862 2.1127>, rotation=<rot -77.5566 177.3915 -172.2686>}]
 foot_l [{shapeType=Capsule, length=10.7147, radius=8.8962, center=<-4.1811 6.0027 -0.338>, rotation=<rot 0.0388 1.4701 -89.3982>}]
 foot_r [{shapeType=Capsule, length=10.7147, radius=8.8962, center=<4.1811 -6.0027 0.338>, rotation=<rot 0.0388 1.4701 90.6018>}]
 head [{shapeType=Capsule, length=0, radius=18, center=<9.0697 0.0746 0>, rotation=<rot 0 -90 0>}]
 lowerarm_l []
 lowerarm_r []
 neck_02 []
 pelvis [{shapeType=Capsule, length=9.6042, radius=11.1441, center=<-2.4471 -2.5289 -0.1981>}]
 spine_01 [{shapeType=Capsule, length=4.3888, radius=14.3638, center=<-8.9021 -0.8973 0>}]
 spine_02 [
  {shapeType=Capsule, length=7.977, radius=16.6037, center=<-2.6885 -1.2641 0>}
  {shapeType=Capsule, length=20.6875, radius=9.3125, center=<-1.8317 1.9362 0>, rotation=<rot 90 0 0>}
 ]
 spine_03 [
  {shapeType=Capsule, length=2.1248, radius=16.7532, center=<0.0574 0.4416 0>}
  {shapeType=Capsule, length=20.7188, radius=9.2812, center=<-1.2163 0.549 0>, rotation=<rot 90 180 -175.0732>}
 ]
 spine_04 [
  {shapeType=Capsule, length=4.5765, radius=14.8228, center=<4.5375 -0.6779 0>}
  {shapeType=Capsule, length=10.0312, radius=9.2812, center=<-1.2326 1.308 0>, rotation=<rot 90 0 0>}
 ]
 spine_05 [{shapeType=Capsule, length=10.2123, radius=11.8227, center=<0.6969 1.91 0>}]
 thigh_l []
 thigh_r []
 upperarm_l []
 upperarm_r []
}
constraints [
 {
  bone1Name=spine_01
  bone2Name=pelvis
  swing1LimitDegrees=45
  swing1Motion=Free
  swing2LimitDegrees=45
  swing2Motion=Free
  twistLimitDegrees=45
  twistMotion=Free
 }
 {
  bone1Name=spine_02
  bone2Name=spine_01
  swing1LimitDegrees=45
  swing1Motion=Free
  swing2LimitDegrees=45
  swing2Motion=Free
  twistLimitDegrees=45
  twistMotion=Free
 }
 {
  bone1Name=spine_03
  bone2Name=spine_02
  swing1LimitDegrees=45
  swing1Motion=Free
  swing2LimitDegrees=45
  swing2Motion=Free
  twistLimitDegrees=45
  twistMotion=Free
 }
 {
  bone1Name=spine_04
  bone2Name=spine_03
  swing1LimitDegrees=45
  swing1Motion=Free
  swing2LimitDegrees=45
  swing2Motion=Free
  twistLimitDegrees=45
  twistMotion=Free
 }
 {
  bone1Name=spine_05
  bone2Name=spine_04
  swing1LimitDegrees=45
  swing1Motion=Free
  swing2LimitDegrees=45
  swing2Motion=Free
  twistLimitDegrees=45
  twistMotion=Free
 }
 {
  bone1Name=clavicle_l
  bone2Name=spine_05
  swing1LimitDegrees=45
  swing1Motion=Free
  swing2LimitDegrees=45
  swing2Motion=Free
  twistLimitDegrees=45
  twistMotion=Free
 }
 {
  bone1Name=upperarm_l
  bone2Name=clavicle_l
  swing1LimitDegrees=45
  swing1Motion=Free
  swing2LimitDegrees=45
  swing2Motion=Free
  twistLimitDegrees=45
  twistMotion=Free
 }
 {
  bone1Name=lowerarm_l
  bone2Name=upperarm_l
  swing1LimitDegrees=45
  swing1Motion=Free
  swing2LimitDegrees=45
  swing2Motion=Free
  twistLimitDegrees=45
  twistMotion=Free
 }
 {
  bone1Name=neck_02
  bone2Name=spine_05
  swing1LimitDegrees=45
  swing1Motion=Free
  swing2LimitDegrees=45
  swing2Motion=Free
  twistLimitDegrees=45
  twistMotion=Free
 }
 {
  bone1Name=head
  bone2Name=neck_02
  swing1LimitDegrees=45
  swing1Motion=Free
  swing2LimitDegrees=45
  swing2Motion=Free
  twistLimitDegrees=45
  twistMotion=Free
 }
 {
  bone1Name=calf_l
  bone2Name=thigh_l
  swing1LimitDegrees=45
  swing1Motion=Free
  swing2LimitDegrees=45
  swing2Motion=Free
  twistLimitDegrees=45
  twistMotion=Free
 }
 {
  bone1Name=thigh_l
  bone2Name=pelvis
  swing1LimitDegrees=45
  swing1Motion=Free
  swing2LimitDegrees=45
  swing2Motion=Free
  twistLimitDegrees=45
  twistMotion=Free
 }
 {
  bone1Name=foot_l
  bone2Name=calf_l
  swing1LimitDegrees=45
  swing1Motion=Free
  swing2LimitDegrees=45
  swing2Motion=Free
  twistLimitDegrees=45
  twistMotion=Free
 }
 {
  bone1Name=lowerarm_r
  bone2Name=upperarm_r
  swing1LimitDegrees=45
  swing1Motion=Free
  swing2LimitDegrees=45
  swing2Motion=Free
  twistLimitDegrees=45
  twistMotion=Free
 }
 {
  bone1Name=upperarm_r
  bone2Name=clavicle_r
  swing1LimitDegrees=45
  swing1Motion=Free
  swing2LimitDegrees=45
  swing2Motion=Free
  twistLimitDegrees=45
  twistMotion=Free
 }
 {
  bone1Name=clavicle_r
  bone2Name=spine_05
  swing1LimitDegrees=45
  swing1Motion=Free
  swing2LimitDegrees=45
  swing2Motion=Free
  twistLimitDegrees=45
  twistMotion=Free
 }
 {
  bone1Name=thigh_r
  bone2Name=pelvis
  swing1LimitDegrees=45
  swing1Motion=Free
  swing2LimitDegrees=45
  swing2Motion=Free
  twistLimitDegrees=45
  twistMotion=Free
 }
 {
  bone1Name=foot_r
  bone2Name=calf_r
  swing1LimitDegrees=45
  swing1Motion=Free
  swing2LimitDegrees=45
  swing2Motion=Free
  twistLimitDegrees=45
  twistMotion=Free
 }
 {
  bone1Name=calf_r
  bone2Name=thigh_r
  swing1LimitDegrees=45
  swing1Motion=Free
  swing2LimitDegrees=45
  swing2Motion=Free
  twistLimitDegrees=45
  twistMotion=Free
 }
 {
  bone1Name=Dyn_skirt_back_01
  bone2Name=pelvis
  swing1LimitDegrees=45
  swing1Motion=Free
  swing2LimitDegrees=45
  swing2Motion=Free
  twistLimitDegrees=45
  twistMotion=Free
 }
 {
  bone1Name=Dyn_skirt_back_02
  bone2Name=Dyn_skirt_back_01
  swing1LimitDegrees=45
  swing1Motion=Free
  swing2LimitDegrees=45
  swing2Motion=Free
  twistLimitDegrees=45
  twistMotion=Free
 }
 {
  bone1Name=Dyn_skirt_back_03
  bone2Name=Dyn_skirt_back_02
  swing1LimitDegrees=45
  swing1Motion=Free
  swing2LimitDegrees=45
  swing2Motion=Free
  twistLimitDegrees=45
  twistMotion=Free
 }
 {
  bone1Name=Dyn_skirt_back_04
  bone2Name=Dyn_skirt_back_03
  swing1LimitDegrees=45
  swing1Motion=Free
  swing2LimitDegrees=45
  swing2Motion=Free
  twistLimitDegrees=45
  twistMotion=Free
 }
 {
  bone1Name=Dyn_skirt_L_06
  bone2Name=pelvis
  swing1LimitDegrees=45
  swing1Motion=Free
  swing2LimitDegrees=45
  swing2Motion=Free
  twistLimitDegrees=45
  twistMotion=Free
 }
 {
  bone1Name=Dyn_skirt_L_07
  bone2Name=Dyn_skirt_L_06
  swing1LimitDegrees=45
  swing1Motion=Free
  swing2LimitDegrees=45
  swing2Motion=Free
  twistLimitDegrees=45
  twistMotion=Free
 }
 {
  bone1Name=Dyn_skirt_L_08
  bone2Name=Dyn_skirt_L_07
  swing1LimitDegrees=45
  swing1Motion=Free
  swing2LimitDegrees=45
  swing2Motion=Free
  twistLimitDegrees=45
  twistMotion=Free
 }
 {
  bone1Name=Dyn_skirt_L_09
  bone2Name=Dyn_skirt_L_08
  swing1LimitDegrees=45
  swing1Motion=Free
  swing2LimitDegrees=45
  swing2Motion=Free
  twistLimitDegrees=45
  twistMotion=Free
 }
 {
  bone1Name=Dyn_skirt_L_01
  bone2Name=pelvis
  swing1LimitDegrees=45
  swing1Motion=Free
  swing2LimitDegrees=45
  swing2Motion=Free
  twistLimitDegrees=45
  twistMotion=Free
 }
 {
  bone1Name=Dyn_skirt_L_02
  bone2Name=Dyn_skirt_L_01
  swing1LimitDegrees=45
  swing1Motion=Free
  swing2LimitDegrees=45
  swing2Motion=Free
  twistLimitDegrees=45
  twistMotion=Free
 }
 {
  bone1Name=Dyn_skirt_L_03
  bone2Name=Dyn_skirt_L_02
  swing1LimitDegrees=45
  swing1Motion=Free
  swing2LimitDegrees=45
  swing2Motion=Free
  twistLimitDegrees=45
  twistMotion=Free
 }
 {
  bone1Name=Dyn_skirt_L_04
  bone2Name=Dyn_skirt_L_03
  swing1LimitDegrees=45
  swing1Motion=Free
  swing2LimitDegrees=45
  swing2Motion=Free
  twistLimitDegrees=45
  twistMotion=Free
 }
 {
  bone1Name=Dyn_skirt_R_05
  bone2Name=Dyn_skirt_R_04
  swing1LimitDegrees=45
  swing1Motion=Free
  swing2LimitDegrees=45
  swing2Motion=Free
  twistLimitDegrees=45
  twistMotion=Free
 }
 {
  bone1Name=Dyn_skirt_R_02
  bone2Name=pelvis
  swing1LimitDegrees=45
  swing1Motion=Free
  swing2LimitDegrees=45
  swing2Motion=Free
  twistLimitDegrees=45
  twistMotion=Free
 }
 {
  bone1Name=Dyn_skirt_R_03
  bone2Name=Dyn_skirt_R_02
  swing1LimitDegrees=45
  swing1Motion=Free
  swing2LimitDegrees=45
  swing2Motion=Free
  twistLimitDegrees=45
  twistMotion=Free
 }
 {
  bone1Name=Dyn_skirt_R_04
  bone2Name=Dyn_skirt_R_03
  swing1LimitDegrees=45
  swing1Motion=Free
  swing2LimitDegrees=45
  swing2Motion=Free
  twistLimitDegrees=45
  twistMotion=Free
 }
 {
  bone1Name=Dyn_cape_back
  bone2Name=spine_04
  swing1LimitDegrees=45
  swing1Motion=Free
  swing2LimitDegrees=45
  swing2Motion=Free
  twistLimitDegrees=45
  twistMotion=Free
 }
]
