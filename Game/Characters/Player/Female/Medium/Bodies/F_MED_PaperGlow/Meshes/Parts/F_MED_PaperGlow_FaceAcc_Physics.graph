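# Physics asset for the face accessory part of the PaperGlow medium female player character, set up for two dangling earrings.
# Each earring bone has a small capsule body constrained to the head with tight swing limits and a zero twist limit.
# Larger capsules on the head, neck and both clavicles approximate the surrounding body.
# The left and right constraints differ slightly in their second swing limit.
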PhysicsAsset /Game/Characters/Player/Female/Medium/Bodies/F_MED_PaperGlow/Meshes/Parts/F_MED_PaperGlow_FaceAcc_Physics
bodies {
 Dyn_Earring_L [{
   shapeType=Capsule
   length=2.6338
   radius=1.1026
   shapeName=Dyn_Earring_L_capsule
   center=<2.4859 0.0007 0.0315>
   rotation=<rot 0 90.3946 -89.2825>
  }]
 Dyn_Earring_R [{
   shapeType=Capsule
   length=2.4402
   radius=1.152
   shapeName=Dyn_Earring_R_capsule
   center=<2.5532 0.0012 -0.0323>
   rotation=<rot 0 -89.6054 -89.2825>
  }]
 clavicle_l [{
   shapeType=Capsule
   length=5.1374
   radius=7
   shapeName=clavicle_l_capsule
   center=<2.2004 -1.7316 2.7456>
   rotation=<rot -49.4216 0.0004 -5.0005>
  }]
 clavicle_r [{shapeType=Capsule, length=5.1879, radius=7, shapeName=clavicle_l_capsule, center=<-2.6453 0 -2.2797>, rotation=<rot -49.245 0.0001 -0.0001>}]
 head [{shapeType=Capsule, length=8.0095, radius=5.974, shapeName=head_capsule, rotation=<rot 90 0 -14.9998>}]
 neck_01 [{shapeType=Capsule, length=6.3796, radius=7, shapeName=neck_01_capsule, center=<-0.6836 -2.5511 -1.3569>, rotation=<rot 0 -15.0001 -5>}]
}
constraints [
 {bone1Name=Dyn_Earring_L, bone2Name=head, swing1LimitDegrees=10, swing2LimitDegrees=25, twistLimitDegrees=0}
 {bone1Name=Dyn_Earring_R, bone2Name=head, swing1LimitDegrees=10, swing2LimitDegrees=20, twistLimitDegrees=0}
]
